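# Physics asset for the arms of the large male FastCheetah body cosmetic, giving its arm feather joints constrained physical motion.
# Capsule bodies cover the pelvis, upper and lower arms, and three feather joints per side.
# Feather 01 hangs from each upper arm and feathers 02 and 03 from each forearm, with tight swing and twist limits.
# The arm chains connect back to the pelvis through constraints with free angular motion.
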PhysicsAsset /BRCosmetics/Characters/Player/Male/Large/Bodies/M_LRG_FastCheetah/Meshes/M_LRG_FastCheetah_ArmPhysics
bodies {
 dfrm_lowerarm_l [{shapeType=Capsule, length=16, radius=8, shapeName=dfrm_lowerarm_l_capsule, center=<8 0 0>, rotation=<rot 90 0 0>}]
 dfrm_lowerarm_r [{shapeType=Capsule, length=16, radius=8, shapeName=dfrm_lowerarm_r_capsule, center=<-8 0 0>, rotation=<rot 90 0 0>}]
 dfrm_upperarm_l [{shapeType=Capsule, length=18, radius=12, shapeName=dfrm_upperarm_l_capsule, center=<12 0 -5>, rotation=<rot 100 0 0>}]
 dfrm_upperarm_r [{shapeType=Capsule, length=18, radius=12, shapeName=dfrm_upperarm_r_capsule, center=<-12 0 5>, rotation=<rot 100 0 0>}]
 drm_ArmFeather_01_l_jnt [{shapeType=Capsule, length=4, radius=3.5, shapeName=drm_ArmFeather_01_l_jnt_capsule, center=<6 0 0>, rotation=<rot 90 0 -10>}]
 drm_ArmFeather_01_r_jnt [{shapeType=Capsule, length=4, radius=3.5, shapeName=drm_ArmFeather_01_r_jnt_capsule, center=<4 0 0>, rotation=<rot 90 0 10>}]
 drm_ArmFeather_02_l_jnt [{shapeType=Capsule, length=6, radius=4, shapeName=drm_ArmFeather_02_l_jnt_capsule, center=<6 0 0>, rotation=<rot 90 0 -20>}]
 drm_ArmFeather_02_r_jnt [{shapeType=Capsule, length=6, radius=4, shapeName=drm_ArmFeather_02_r_jnt_capsule, center=<6 0 0>, rotation=<rot 90 0 20>}]
 drm_ArmFeather_03_l_jnt [{shapeType=Capsule, length=4, radius=4, shapeName=drm_ArmFeather_03_l_jnt_capsule, center=<6 0 0>, rotation=<rot 90 0 -20>}]
 drm_ArmFeather_03_r_jnt [{shapeType=Capsule, length=4, radius=4, shapeName=drm_ArmFeather_03_r_jnt_capsule, center=<6 0 0>, rotation=<rot 90 0 20>}]
 pelvis [{shapeType=Capsule, length=4.9046, radius=35.2946, center=<-10.9162 -4.9547 -0.1765>, rotation=<rot 0.3009 -90 2.6868>}]
}
constraints [
 {bone1Name=drm_ArmFeather_02_l_jnt, bone2Name=dfrm_lowerarm_l, swing1LimitDegrees=6, swing2LimitDegrees=6, twistLimitDegrees=2}
 {bone1Name=drm_ArmFeather_03_l_jnt, bone2Name=dfrm_lowerarm_l, swing1LimitDegrees=6, swing2LimitDegrees=6, twistLimitDegrees=2}
 {bone1Name=drm_ArmFeather_01_l_jnt, bone2Name=dfrm_upperarm_l, swing1LimitDegrees=6, swing2LimitDegrees=6, twistLimitDegrees=2}
 {bone1Name=drm_ArmFeather_02_r_jnt, bone2Name=dfrm_lowerarm_r, swing1LimitDegrees=6, swing2LimitDegrees=6, twistLimitDegrees=2}
 {bone1Name=drm_ArmFeather_03_r_jnt, bone2Name=dfrm_lowerarm_r, swing1LimitDegrees=6, swing2LimitDegrees=6, twistLimitDegrees=2}
 {bone1Name=drm_ArmFeather_01_r_jnt, bone2Name=dfrm_upperarm_r, swing1LimitDegrees=6, swing2LimitDegrees=6, twistLimitDegrees=2}
 {
  bone1Name=dfrm_upperarm_r
  bone2Name=pelvis
  swing1LimitDegrees=45
  swing1Motion=Free
  swing2LimitDegrees=45
  swing2Motion=Free
  twistLimitDegrees=45
  twistMotion=Free
 }
 {
  bone1Name=dfrm_lowerarm_r
  bone2Name=dfrm_upperarm_r
  swing1LimitDegrees=45
  swing1Motion=Free
  swing2LimitDegrees=45
  swing2Motion=Free
  twistLimitDegrees=45
  twistMotion=Free
 }
 {
  bone1Name=dfrm_lowerarm_l
  bone2Name=dfrm_upperarm_l
  swing1LimitDegrees=45
  swing1Motion=Free
  swing2LimitDegrees=45
  swing2Motion=Free
  twistLimitDegrees=45
  twistMotion=Free
 }
 {
  bone1Name=dfrm_upperarm_l
  bone2Name=pelvis
  swing1LimitDegrees=45
  swing1Motion=Free
  swing2LimitDegrees=45
  swing2Motion=Free
  twistLimitDegrees=45
  twistMotion=Free
 }
]
