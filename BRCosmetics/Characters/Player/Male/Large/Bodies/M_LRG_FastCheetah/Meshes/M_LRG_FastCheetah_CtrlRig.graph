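# Control Rig for the large male FastCheetah player body, correcting deformation of arms and stomach and repositioning the backpack attach bone.
# Spherical pose readers and bone-distance remaps produce blend alphas.
# Deform arms copy source rotations, use PBIK to match hands, then receive roll and forearm twist.
# Alphas drive belly offsets alongside spatial clamps, a shared DeformRig function, and pose-read stomach and backpack adjustments gated by aiming or melee state.
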ControlRigBlueprint /BRCosmetics/Characters/Player/Male/Large/Bodies/M_LRG_FastCheetah/Meshes/M_LRG_FastCheetah_CtrlRig
graphs [
 Multiply_1_1_1_ContainedGraph {nodes [
   n1 = RigVMFunctionEntryNode()
   RigVMFunctionReturnNode(Result=&RigVMFunction_MathDoubleMul(B=@n1.C, A=&RigVMFunction_MathDoubleMul(A=@n1.A, B=@n1.B).Result).Result)
  ]}
 Multiply_1_ContainedGraph {nodes [
   n1 = RigVMFunctionEntryNode()
   Multiply = &RigVMFunction_MathDoubleMul(A=@n1.A, B=@n1.B)
   RigVMFunctionReturnNode(Result=&RigVMFunction_MathDoubleMul(B=@n1.C, A=@Multiply.Result).Result)
  ]}
 RigVMFunctionLibrary {links [], nodes []}
 RigVMFunction_Sequence_1_1_ContainedGraph {nodes [
   n1 = RigVMFunctionEntryNode()
   n2 = &RigVMFunction_Sequence(ExecuteContext=@n1.ExecuteContext)
   n3 = &RigVMFunction_Sequence(ExecuteContext=@n2.B)
   n4 = &RigVMFunction_Sequence(ExecuteContext=@n3.B)
   RigVMFunctionReturnNode(A=@n2.A, B=@n3.A, C=@n4.A, D=@n4.B)
  ]}
 RigVMModel {nodes [
   RigUnit_BeginExecution = &RigUnit_BeginExecution()
   SequenceAggregate_1 = RigVMAggregateNode(ExecuteContext=@RigUnit_BeginExecution.ExecutePin)
   n1 = &RigUnit_SphericalPoseReader(
    ActiveRegionScaleFactors=(PositiveWidth=1,NegativeWidth=1,PositiveHeight=1.2,NegativeHeight=1.2)
    ActiveRegionSize=0.05
    Debug=(bDrawDebug=false,bDraw2D=False,bDrawLocalAxes=False,DebugScale=25,DebugSegments=20,DebugThickness=0.25)
    DriverAxis=(X=1,Y=0,Z=0)
    DriverItem=(Type=Bone,Name="upperarm_l")
    FalloffRegionScaleFactors=(PositiveWidth=0.5,NegativeWidth=0.75,PositiveHeight=0.8,NegativeHeight=0.5)
    FalloffSize=0.3
    FlipHeightScaling=False
    FlipWidthScaling=False
    OptionalParentItem=(Type=Bone,Name="None")
    RotationOffset=(X=0,Y=-155,Z=27.5)
    ExecutePin=@SequenceAggregate_1.A
   )
   > RigVMVariableNode(Value=@n1.OutputParam) {variable=SPR_upperarm_l_down_A_alpha}
   > n2 = &RigUnit_SphericalPoseReader(
    ActiveRegionScaleFactors=(PositiveWidth=1,NegativeWidth=1,PositiveHeight=1,NegativeHeight=1)
    ActiveRegionSize=0.065
    Debug=(bDrawDebug=false,bDraw2D=False,bDrawLocalAxes=False,DebugScale=25,DebugSegments=20,DebugThickness=0.25)
    DriverAxis=(X=-1,Y=0,Z=0)
    DriverItem=(Type=Bone,Name="upperarm_r")
    FalloffRegionScaleFactors=(PositiveWidth=0.75,NegativeWidth=1,PositiveHeight=0.75,NegativeHeight=1)
    FalloffSize=0.2
    FlipHeightScaling=False
    FlipWidthScaling=False
    OptionalParentItem=(Type=Bone,Name="None")
    RotationOffset=(X=0,Y=50,Z=-7.5)
   )
   > RigVMVariableNode(Value=@n2.OutputParam) {variable=SPR_upperarm_r_down_A_alpha}
   > n3 = &RigUnit_SphericalPoseReader(
    ActiveRegionScaleFactors=(PositiveWidth=1,NegativeWidth=1,PositiveHeight=1,NegativeHeight=1)
    ActiveRegionSize=0.05
    Debug=(bDrawDebug=false,bDraw2D=False,bDrawLocalAxes=false,DebugScale=25,DebugSegments=20,DebugThickness=0.25)
    DriverAxis=(X=-1,Y=0,Z=0)
    DriverItem=(Type=Bone,Name="upperarm_r")
    FalloffRegionScaleFactors=(PositiveWidth=0.5,NegativeWidth=1,PositiveHeight=0.75,NegativeHeight=1)
    FalloffSize=0.2
    FlipHeightScaling=False
    FlipWidthScaling=False
    OptionalParentItem=(Type=Bone,Name="None")
    RotationOffset=(X=-23.5,Y=17.5,Z=0)
   )
   > RigVMVariableNode(Value=@n3.OutputParam) {variable=SPR_upperarm_r_down_B_alpha}
   SequenceAggregate = &RigVMFunction_Sequence(ExecuteContext=@SequenceAggregate_1.C)
   RigUnit_ItemArray = &RigUnit_ItemArray(Items=((Type=Bone,Name="clavicle_l"),(Type=Bone,Name="upperarm_l"),(Type=Bone,Name="lowerarm_l")))
   DISPATCH_RigVMDispatch_ArrayIterator = RigVMDispatchNode(Array=@RigUnit_ItemArray.Items, ExecuteContext=@SequenceAggregate.A) {template="DISPATCH_RigVMDispatch_ArrayIterator(in Array,out Element,out Index,out Count,out Ratio)"}
   DISPATCH_RigVMDispatch_ArrayGetAtIndex = RigVMDispatchNode(
    Array=&RigUnit_ItemArray(Items=((Type=Bone,Name="dfrm_clavicle_l"),(Type=Bone,Name="dfrm_upperarm_l"),(Type=Bone,Name="dfrm_lowerarm_l"))).Items
    Index=@DISPATCH_RigVMDispatch_ArrayIterator.Index
   ) {template="DISPATCH_RigVMDispatch_ArrayGetAtIndex(in Array,in Index,out Element)"}
   > &RigUnit_SetRotation(
    Space=GlobalSpace
    Weight=1
    bInitial=False
    bPropagateToChildren=true
    Item=@DISPATCH_RigVMDispatch_ArrayGetAtIndex.Element
    Value=&RigUnit_GetTransform(Item=@DISPATCH_RigVMDispatch_ArrayIterator.Element).Transform.Rotation
   )
   PBIK = /Script/PBIK.RigUnit_PBIK(
    BoneSettings=((Bone="dfrm_lowerarm_l",X=Locked,Y=Locked,Z=Free,PreferredAngles=(X=0,Y=0,Z=0)))
    Debug=(DrawScale=1)
    Effectors=((Bone="dfrm_hand_l",Transform=(Rotation=(X=0,Y=0,Z=0,W=1),Translation=(X=0,Y=0,Z=0),Scale3D=(X=1,Y=1,Z=1)),PositionAlpha=1,RotationAlpha=1,StrengthAlpha=1,PinRotation=1,ChainDepth=0))
    ExcludedBones=()
    Root=dfrm_clavicle_l
    Settings="(Iterations=20,SubIterations=0,MassMultiplier=1,bAllowStretch=False,RootBehavior=PinToInput,PrePullRootSettings=(RotationAlpha=0,RotationAlphaX=1,RotationAlphaY=1,RotationAlphaZ=1,PositionAlpha=1,PositionAlphaX=1,PositionAlphaY=1,PositionAlphaZ=1),GlobalPu ...(truncated)"
    Effectors.0.Transform=&RigUnit_GetTransform(Item=(Type=Bone,Name="hand_l")).Transform
    ExecutePin=@DISPATCH_RigVMDispatch_ArrayIterator.Completed
   ) {is=struct RigVMUnitNode}
   n4 = RigVMDispatchNode(
    Array=&RigUnit_ItemArray(Items=((Type=Bone,Name="clavicle_r"),(Type=Bone,Name="upperarm_r"),(Type=Bone,Name="lowerarm_r"))).Items
    ExecuteContext=@SequenceAggregate.B
   ) {template="DISPATCH_RigVMDispatch_ArrayIterator(in Array,out Element,out Index,out Count,out Ratio)"}
   > &RigUnit_SetRotation(
    Space=GlobalSpace
    Weight=1
    bInitial=False
    bPropagateToChildren=true
    Item=RigVMDispatchNode(
     Array=&RigUnit_ItemArray(Items=((Type=Bone,Name="dfrm_clavicle_r"),(Type=Bone,Name="dfrm_upperarm_r"),(Type=Bone,Name="dfrm_lowerarm_r"))).Items
     Index=@n4.Index
    ) {template="DISPATCH_RigVMDispatch_ArrayGetAtIndex(in Array,in Index,out Element)"}.Element
    Value=&RigUnit_GetTransform(Item=@n4.Element).Transform.Rotation
   )
   &RigUnit_PBIK(
    BoneSettings=((Bone="dfrm_lowerarm_r",X=Locked,Y=Locked,Z=Free,PreferredAngles=(X=0,Y=0,Z=0)))
    Debug=(DrawScale=1)
    Effectors=((Bone="dfrm_hand_r",Transform=(Rotation=(X=0,Y=0,Z=0,W=1),Translation=(X=0,Y=0,Z=0),Scale3D=(X=1,Y=1,Z=1)),PositionAlpha=1,RotationAlpha=1,StrengthAlpha=1,PinRotation=1,ChainDepth=0))
    ExcludedBones=()
    Root=dfrm_clavicle_r
    Settings="(Iterations=20,SubIterations=0,MassMultiplier=1,bAllowStretch=False,RootBehavior=PinToInput,PrePullRootSettings=(RotationAlpha=0,RotationAlphaX=1,RotationAlphaY=1,RotationAlphaZ=1,PositionAlpha=1,PositionAlphaX=1,PositionAlphaY=1,PositionAlphaZ=1),GlobalPu ...(truncated)"
    Effectors.0.Transform=&RigUnit_GetTransform(Item=(Type=Bone,Name="hand_r")).Transform
    ExecutePin=@n4.Completed
   )
   SequenceAggregate_2 = &RigVMFunction_Sequence(ExecuteContext=@SequenceAggregate_1.D)
   Subtract = /Script/RigVM.RigVMFunction_MathVectorSub(
    B=&RigUnit_GetTransform(Item=(Type=Bone,Name="dfrm_upperarm_l")).Transform.Translation
    A=&RigUnit_GetTransform(Item=(Type=Bone,Name="dfrm_hand_l")).Transform.Translation
   ) {is=struct RigVMUnitNode, template="Subtract::Execute(in A,in B,out Result)"}
   Unit = /Script/RigVM.RigVMFunction_MathVectorUnit(Value=@Subtract.Result) {is=struct RigVMUnitNode, template="Unit::Execute(in Value,out Result)"}
   RigVMFunction_MathQuaternionFromAxisAndAngle = /Script/RigVM.RigVMFunction_MathQuaternionFromAxisAndAngle(Angle=-1, Axis=@Unit.Result) {is=struct RigVMUnitNode, template="FromAxisAndAngle::Execute(in Axis,in Angle,out Result)"}
   Set Transform = &RigUnit_SetRotation(
    Item=(Type=Bone,Name="dfrm_upperarm_l")
    Space=GlobalSpace
    bInitial=false
    Value=/Script/RigVM.RigVMFunction_MathQuaternionMul(
     B=&RigUnit_GetTransform(Item=(Type=Bone,Name="dfrm_upperarm_l")).Transform.Rotation
     A=@RigVMFunction_MathQuaternionFromAxisAndAngle.Result
    ) {is=struct RigVMUnitNode, template="Multiply::Execute(in A,in B,out Result)"}.Result
    ExecutePin=@SequenceAggregate_2.A
    Weight=RigVMVariableNode() {variable=mult_spr_dst_lowerarm_r_pelvis_alpha}.Value
   )
   n5 = &RigVMFunction_MathVectorUnit(Value=&RigVMFunction_MathVectorSub(
     B=&RigUnit_GetTransform(Item=(Type=Bone,Name="dfrm_upperarm_r"), bInitial=False).Transform.Translation
     A=&RigUnit_GetTransform(Item=(Type=Bone,Name="dfrm_hand_r"), bInitial=False).Transform.Translation
    ) {template="Subtract::Execute(in A,in B,out Result)"}.Result) {template="Unit::Execute(in Value,out Result)"}
   Set Transform_1 = &RigUnit_SetRotation(
    Item=(Type=Bone,Name="dfrm_upperarm_r")
    Space=GlobalSpace
    bInitial=False
    Value=&RigVMFunction_MathQuaternionMul(
     B=&RigUnit_GetTransform(Item=(Type=Bone,Name="dfrm_upperarm_r"), bInitial=False).Transform.Rotation
     A=&RigVMFunction_MathQuaternionFromAxisAndAngle(Angle=0.35, Axis=@n5.Result) {template="FromAxisAndAngle::Execute(in Axis,in Angle,out Result)"}.Result
    ) {template="Multiply::Execute(in A,in B,out Result)"}.Result
    Weight=RigVMVariableNode() {variable=SPR_upperarm_r_down_B_alpha}.Value
    ExecutePin=@SequenceAggregate_2.B
   )
   FN_DeformRig_FNC = RigVMFunctionReferenceNode(
    DeformRig_ Curve="(deformRig_l_upperarm_up_alpha_72_#1=False,deformRig_r_upperarm_up_alpha_73_#2=False,deformRig_l_upperarm_fr_alpha_74_#3=False,deformRig_r_upperarm_fr_alpha_75_CB51B8 ...(truncated)"
    DeformRig_Bool="(upperarm_1_#4=True,elbow_3_#5=False,wrist_7_#6=False,groin_8_#7=False,glute_17_#8=False,knee_9_40C408324C ...(truncated)"
    DeformRig_Debug=(upperarm_debug_18_#4=False,wrist_debug_20_#6=False,lat_debug_26_#9=False,groinGlute_debug_27_#7=False)
    DeformRig_Value="(upperarm_up_18_#10=(Rotation=(X=0.0,Y=-0.0,Z=0.0,W=1.0),Translation=(X=0.0,Y=0.0,Z=8.0),Scale3D=(X=0.70,Y=1.0,Z=1.0)),upperarm_fwd_fr_19_#11=(Rotation=(X=0.0,Y=-0.0,Z=0.0,W=1.0),Translation=(X=0.0, ...(truncated)"
    ExecuteContext=@SequenceAggregate_1.G
   )
   > n6 = &RigUnit_SphericalPoseReader(
    ActiveRegionScaleFactors=(PositiveWidth=1,NegativeWidth=1,PositiveHeight=0.5,NegativeHeight=1)
    ActiveRegionSize=0.1
    Debug=(bDrawDebug=false,bDraw2D=false,bDrawLocalAxes=false,DebugScale=25,DebugSegments=20,DebugThickness=0.25)
    DriverAxis=(X=1,Y=0,Z=0)
    DriverItem=(Type=Bone,Name="upperarm_l")
    FalloffRegionScaleFactors=(PositiveWidth=0.25,NegativeWidth=0.25,PositiveHeight=0.1,NegativeHeight=0.25)
    FalloffSize=0.25
    OptionalParentItem=(Type=Bone,Name="clavicle_l")
    RotationOffset=(X=20,Y=210,Z=-35)
   )
   n7 = RigVMVariableNode() {variable=Is Rifle Targeting Alpha}
   Multiply_1_1 = RigVMAggregateNode(C=1, A=@n6.OutputParam, B=@n7.Value)
   AlphaInterp = &RigVMFunction_AlphaInterp(InterpSpeedDecreasing=7, InterpSpeedIncreasing=10, Scale=1, bClampResult=False, Value=@Multiply_1_1.Result)
   > &RigUnit_ModifyTransforms(
    ItemToModify=((Item=(Type=Bone,Name="dfrm_stomach_mid"),Transform=(Rotation=(X=0,Y=0,Z=0,W=1),Translation=(X=-20,Y=0,Z=0),Scale3D=(X=1,Y=1,Z=1))))
    Weight=@AlphaInterp.Result
    ItemToModify.0.Transform.Rotation=&RigVMFunction_MathQuaternionFromEuler(Euler=(X=0,Y=0,Z=0)).Result
   )
   > &RigUnit_ModifyTransforms(
    ItemToModify=((Item=(Type=Bone,Name="dfrm_stomach_lower_mid"),Transform=(Rotation=(X=0,Y=0,Z=0,W=1),Translation=(X=-12,Y=0,Z=0),Scale3D=(X=1,Y=1,Z=1))))
    Weight=@AlphaInterp.Result
    ItemToModify.0.Transform.Rotation=&RigVMFunction_MathQuaternionFromEuler(Euler=(X=0,Y=0,Z=0)).Result
   )
   > &RigUnit_ModifyTransforms(
    ItemToModify=((Item=(Type=Bone,Name="dfrm_stomach_01_l"),Transform=(Rotation=(X=0,Y=0,Z=0,W=1),Translation=(X=-8,Y=0,Z=0),Scale3D=(X=1,Y=1,Z=1))))
    Weight=@AlphaInterp.Result
    ItemToModify.0.Transform.Rotation=&RigVMFunction_MathQuaternionFromEuler(Euler=(X=0,Y=0,Z=0)).Result
   )
   > &RigUnit_ModifyTransforms(
    ItemToModify=((Item=(Type=Bone,Name="dfrm_stomach_04_l"),Transform=(Rotation=(X=0,Y=0,Z=0,W=1),Translation=(X=-16,Y=0,Z=0),Scale3D=(X=1,Y=1,Z=1))))
    Weight=@AlphaInterp.Result
    ItemToModify.0.Transform.Rotation=&RigVMFunction_MathQuaternionFromEuler(Euler=(X=0,Y=0,Z=0)).Result
   )
   > &RigUnit_ModifyTransforms(
    ItemToModify=((Item=(Type=Bone,Name="dfrm_stomach_01_r"),Transform=(Rotation=(X=0,Y=0,Z=0,W=1),Translation=(X=-8,Y=0,Z=0),Scale3D=(X=1,Y=1,Z=1))))
    Weight=@AlphaInterp.Result
    ItemToModify.0.Transform.Rotation=&RigVMFunction_MathQuaternionFromEuler(Euler=(X=0,Y=0,Z=0)).Result
   )
   > &RigUnit_ModifyTransforms(
    ItemToModify=((Item=(Type=Bone,Name="dfrm_stomach_04_r"),Transform=(Rotation=(X=0,Y=0,Z=0,W=1),Translation=(X=-16,Y=0,Z=0),Scale3D=(X=1,Y=1,Z=1))))
    Weight=@AlphaInterp.Result
    ItemToModify.0.Transform.Rotation=&RigVMFunction_MathQuaternionFromEuler(Euler=(X=0,Y=0,Z=0)).Result
   )
   > n8 = &RigUnit_SphericalPoseReader(
    ActiveRegionScaleFactors=(PositiveWidth=0.5,NegativeWidth=0.5,PositiveHeight=0.5,NegativeHeight=0.5)
    ActiveRegionSize=0.1
    Debug=(bDrawDebug=false,bDraw2D=false,bDrawLocalAxes=false,DebugScale=25,DebugSegments=20,DebugThickness=0.25)
    DriverAxis=(X=-1,Y=0,Z=0)
    DriverItem=(Type=Bone,Name="upperarm_r")
    FalloffRegionScaleFactors=(PositiveWidth=0.25,NegativeWidth=0.25,PositiveHeight=0.25,NegativeHeight=0.25)
    FalloffSize=0.25
    OptionalParentItem=(Type=Bone,Name="clavicle_r")
    RotationOffset=(X=-35,Y=20,Z=0)
   )
   n9 = RigVMVariableNode() {variable=Is Pistol Targeting Alpha}
   Multiply_5 = RigVMAggregateNode(C=1, A=@n8.OutputParam, B=@n9.Value)
   n10 = &RigVMFunction_AlphaInterp(InterpSpeedDecreasing=7, InterpSpeedIncreasing=10, Scale=1, bClampResult=False, Value=@Multiply_5.Result)
   > &RigUnit_ModifyTransforms(
    ItemToModify=((Item=(Type=Bone,Name="dfrm_stomach_mid"),Transform=(Rotation=(X=0,Y=0,Z=0,W=1),Translation=(X=-20,Y=0,Z=0),Scale3D=(X=1,Y=1,Z=1))))
    Weight=@n10.Result
    ItemToModify.0.Transform.Rotation=&RigVMFunction_MathQuaternionFromEuler(Euler=(X=0,Y=0,Z=0)).Result
   )
   > &RigUnit_ModifyTransforms(
    ItemToModify=((Item=(Type=Bone,Name="dfrm_stomach_lower_mid"),Transform=(Rotation=(X=0,Y=0,Z=0,W=1),Translation=(X=0,Y=0,Z=0),Scale3D=(X=1,Y=1,Z=1))))
    Weight=@n10.Result
    ItemToModify.0.Transform.Rotation=&RigVMFunction_MathQuaternionFromEuler(Euler=(X=0,Y=0,Z=0)).Result
   )
   > n11 = &RigUnit_SphericalPoseReader(
    ActiveRegionScaleFactors=(PositiveWidth=0.75,NegativeWidth=0.75,PositiveHeight=1,NegativeHeight=1)
    ActiveRegionSize=0.1
    Debug=(bDrawDebug=false,bDraw2D=false,bDrawLocalAxes=false,DebugScale=25,DebugSegments=20,DebugThickness=0.25)
    DriverAxis=(X=0,Y=1,Z=0)
    DriverItem=(Type=Bone,Name="spine_05")
    FalloffRegionScaleFactors=(PositiveWidth=0.5,NegativeWidth=0.5,PositiveHeight=0.5,NegativeHeight=0.5)
    FalloffSize=0.25
    OptionalParentItem=(Type=Bone,Name="spine_01")
    RotationOffset=(X=90,Y=45,Z=20)
   )
   n12 = &RigVMFunction_AlphaInterp(
    InterpSpeedDecreasing=7
    InterpSpeedIncreasing=10
    Scale=1
    bClampResult=False
    Value=&RigVMFunction_MathDoubleMul(B=@n11.OutputParam, A=@n7.Value).Result
   )
   > &RigUnit_ModifyTransforms(
    ItemToModify=((Item=(Type=Bone,Name="dfrm_stomach_01_l"),Transform=(Rotation=(X=0,Y=0,Z=0,W=1),Translation=(X=-16,Y=0,Z=0),Scale3D=(X=1,Y=1,Z=1))))
    Weight=@n12.Result
    ItemToModify.0.Transform.Rotation=&RigVMFunction_MathQuaternionFromEuler(Euler=(X=0,Y=0,Z=0)).Result
   )
   > &RigUnit_ModifyTransforms(
    ItemToModify=((Item=(Type=Bone,Name="dfrm_stomach_04_l"),Transform=(Rotation=(X=0,Y=0,Z=0,W=1),Translation=(X=-8,Y=0,Z=0),Scale3D=(X=1,Y=1,Z=1))))
    Weight=@n12.Result
    ItemToModify.0.Transform.Rotation=&RigVMFunction_MathQuaternionFromEuler(Euler=(X=0,Y=0,Z=0)).Result
   )
   > &RigUnit_ModifyTransforms(
    ItemToModify=((Item=(Type=Bone,Name="dfrm_stomach_02_l"),Transform=(Rotation=(X=0,Y=0,Z=0,W=1),Translation=(X=-16,Y=0,Z=0),Scale3D=(X=1,Y=1,Z=1))))
    Weight=@n12.Result
    ItemToModify.0.Transform.Rotation=&RigVMFunction_MathQuaternionFromEuler(Euler=(X=0,Y=0,Z=0)).Result
   )
   > &RigUnit_ModifyTransforms(
    ItemToModify=((Item=(Type=Bone,Name="dfrm_stomach_05_l"),Transform=(Rotation=(X=0,Y=0,Z=0,W=1),Translation=(X=-8,Y=0,Z=0),Scale3D=(X=1,Y=1,Z=1))))
    Weight=@n12.Result
    ItemToModify.0.Transform.Rotation=&RigVMFunction_MathQuaternionFromEuler(Euler=(X=0,Y=0,Z=0)).Result
   )
   > &RigUnit_ModifyTransforms(
    ItemToModify=((Item=(Type=Bone,Name="dfrm_stomach_03_l"),Transform=(Rotation=(X=0,Y=0,Z=0,W=1),Translation=(X=-4,Y=0,Z=0),Scale3D=(X=1,Y=1,Z=1))))
    Weight=@n12.Result
    ItemToModify.0.Transform.Rotation=&RigVMFunction_MathQuaternionFromEuler(Euler=(X=0,Y=0,Z=0)).Result
   )
   > &RigUnit_ModifyTransforms(
    ItemToModify=((Item=(Type=Bone,Name="dfrm_stomach_06_l"),Transform=(Rotation=(X=0,Y=0,Z=0,W=1),Translation=(X=-4,Y=0,Z=0),Scale3D=(X=1,Y=1,Z=1))))
    Weight=@n12.Result
    ItemToModify.0.Transform.Rotation=&RigVMFunction_MathQuaternionFromEuler(Euler=(X=0,Y=0,Z=0)).Result
   )
   > n13 = &RigUnit_SphericalPoseReader(
    ActiveRegionScaleFactors=(PositiveWidth=0.75,NegativeWidth=0.75,PositiveHeight=0.75,NegativeHeight=0.75)
    ActiveRegionSize=0.1
    Debug=(bDrawDebug=false,bDraw2D=false,bDrawLocalAxes=false,DebugScale=25,DebugSegments=20,DebugThickness=0.25)
    DriverAxis=(X=0,Y=1,Z=0)
    DriverItem=(Type=Bone,Name="spine_05")
    FalloffRegionScaleFactors=(PositiveWidth=0.25,NegativeWidth=0.25,PositiveHeight=0.25,NegativeHeight=0.25)
    FalloffSize=0.45
    OptionalParentItem=(Type=Bone,Name="spine_01")
    RotationOffset=(X=120,Y=-60,Z=0)
   )
   n14 = &RigVMFunction_AlphaInterp(
    InterpSpeedDecreasing=7
    InterpSpeedIncreasing=10
    Scale=1
    bClampResult=False
    Value=&RigVMFunction_MathDoubleMul(B=@n13.OutputParam, A=@n9.Value).Result
   )
   > &RigUnit_ModifyTransforms(
    ItemToModify=((Item=(Type=Bone,Name="dfrm_stomach_01_l"),Transform=(Rotation=(X=0,Y=0,Z=0,W=1),Translation=(X=-12,Y=0,Z=0),Scale3D=(X=1,Y=1,Z=1))))
    Weight=@n14.Result
    ItemToModify.0.Transform.Rotation=&RigVMFunction_MathQuaternionFromEuler(Euler=(X=0,Y=0,Z=0)).Result
   )
   > &RigUnit_ModifyTransforms(
    ItemToModify=((Item=(Type=Bone,Name="dfrm_stomach_04_l"),Transform=(Rotation=(X=0,Y=0,Z=0,W=1),Translation=(X=-12,Y=0,Z=0),Scale3D=(X=1,Y=1,Z=1))))
    Weight=@n14.Result
    ItemToModify.0.Transform.Rotation=&RigVMFunction_MathQuaternionFromEuler(Euler=(X=0,Y=0,Z=0)).Result
   )
   > &RigUnit_ModifyTransforms(
    ItemToModify=((Item=(Type=Bone,Name="dfrm_stomach_02_l"),Transform=(Rotation=(X=0,Y=0,Z=0,W=1),Translation=(X=-12,Y=0,Z=0),Scale3D=(X=1,Y=1,Z=1))))
    Weight=@n14.Result
    ItemToModify.0.Transform.Rotation=&RigVMFunction_MathQuaternionFromEuler(Euler=(X=0,Y=0,Z=0)).Result
   )
   > &RigUnit_ModifyTransforms(
    ItemToModify=((Item=(Type=Bone,Name="dfrm_stomach_05_l"),Transform=(Rotation=(X=0,Y=0,Z=0,W=1),Translation=(X=-12,Y=0,Z=0),Scale3D=(X=1,Y=1,Z=1))))
    Weight=@n14.Result
    ItemToModify.0.Transform.Rotation=&RigVMFunction_MathQuaternionFromEuler(Euler=(X=0,Y=0,Z=0)).Result
   )
   > &RigUnit_ModifyTransforms(
    ItemToModify=((Item=(Type=Bone,Name="dfrm_stomach_03_l"),Transform=(Rotation=(X=0,Y=0,Z=0,W=1),Translation=(X=-4,Y=0,Z=0),Scale3D=(X=1,Y=1,Z=1))))
    Weight=@n14.Result
    ItemToModify.0.Transform.Rotation=&RigVMFunction_MathQuaternionFromEuler(Euler=(X=0,Y=0,Z=0)).Result
   )
   > &RigUnit_ModifyTransforms(
    ItemToModify=((Item=(Type=Bone,Name="dfrm_stomach_06_l"),Transform=(Rotation=(X=0,Y=0,Z=0,W=1),Translation=(X=-4,Y=0,Z=0),Scale3D=(X=1,Y=1,Z=1))))
    Weight=@n14.Result
    ItemToModify.0.Transform.Rotation=&RigVMFunction_MathQuaternionFromEuler(Euler=(X=0,Y=0,Z=0)).Result
   )
   > n15 = &RigUnit_SphericalPoseReader(
    ActiveRegionScaleFactors=(PositiveWidth=0.75,NegativeWidth=0.75,PositiveHeight=1,NegativeHeight=1)
    ActiveRegionSize=0.1
    Debug=(bDrawDebug=false,bDraw2D=false,bDrawLocalAxes=false,DebugScale=25,DebugSegments=20,DebugThickness=0.25)
    DriverAxis=(X=0,Y=1,Z=0)
    DriverItem=(Type=Bone,Name="spine_05")
    FalloffRegionScaleFactors=(PositiveWidth=0.5,NegativeWidth=0.5,PositiveHeight=0.5,NegativeHeight=0.5)
    FalloffSize=0.25
    OptionalParentItem=(Type=Bone,Name="spine_01")
    RotationOffset=(X=0,Y=-30,Z=110)
   )
   n16 = &RigVMFunction_AlphaInterp(
    InterpSpeedDecreasing=7
    InterpSpeedIncreasing=10
    Scale=1
    bClampResult=False
    Value=&RigVMFunction_MathDoubleMul(B=@n15.OutputParam, A=@n7.Value).Result
   )
   > &RigUnit_ModifyTransforms(
    ItemToModify=((Item=(Type=Bone,Name="dfrm_stomach_01_r"),Transform=(Rotation=(X=0,Y=0,Z=0,W=1),Translation=(X=-16,Y=0,Z=0),Scale3D=(X=1,Y=1,Z=1))))
    Weight=@n16.Result
    ItemToModify.0.Transform.Rotation=&RigVMFunction_MathQuaternionFromEuler(Euler=(X=0,Y=0,Z=0)).Result
   )
   > &RigUnit_ModifyTransforms(
    ItemToModify=((Item=(Type=Bone,Name="dfrm_stomach_04_r"),Transform=(Rotation=(X=0,Y=0,Z=0,W=1),Translation=(X=-8,Y=0,Z=0),Scale3D=(X=1,Y=1,Z=1))))
    Weight=@n16.Result
    ItemToModify.0.Transform.Rotation=&RigVMFunction_MathQuaternionFromEuler(Euler=(X=0,Y=0,Z=0)).Result
   )
   > n17 = &RigUnit_SphericalPoseReader(
    ActiveRegionScaleFactors=(PositiveWidth=0.75,NegativeWidth=0.75,PositiveHeight=0.75,NegativeHeight=0.75)
    ActiveRegionSize=0.1
    Debug=(bDrawDebug=false,bDraw2D=false,bDrawLocalAxes=false,DebugScale=25,DebugSegments=20,DebugThickness=0.25)
    DriverAxis=(X=0,Y=1,Z=0)
    DriverItem=(Type=Bone,Name="spine_05")
    FalloffRegionScaleFactors=(PositiveWidth=0.25,NegativeWidth=0.25,PositiveHeight=0.25,NegativeHeight=0.25)
    FalloffSize=0.45
    OptionalParentItem=(Type=Bone,Name="spine_01")
    RotationOffset=(X=140,Y=-150,Z=0)
   )
   n18 = &RigVMFunction_AlphaInterp(
    InterpSpeedDecreasing=7
    InterpSpeedIncreasing=10
    Scale=1
    bClampResult=False
    Value=&RigVMFunction_MathDoubleMul(B=@n17.OutputParam, A=@n9.Value).Result
   )
   > &RigUnit_ModifyTransforms(
    ItemToModify=((Item=(Type=Bone,Name="dfrm_stomach_01_r"),Transform=(Rotation=(X=0,Y=0,Z=0,W=1),Translation=(X=-12,Y=0,Z=0),Scale3D=(X=1,Y=1,Z=1))))
    Weight=@n18.Result
    ItemToModify.0.Transform.Rotation=&RigVMFunction_MathQuaternionFromEuler(Euler=(X=0,Y=0,Z=0)).Result
   )
   > &RigUnit_ModifyTransforms(
    ItemToModify=((Item=(Type=Bone,Name="dfrm_stomach_04_r"),Transform=(Rotation=(X=0,Y=0,Z=0,W=1),Translation=(X=-12,Y=0,Z=0),Scale3D=(X=1,Y=1,Z=1))))
    Weight=@n18.Result
    ItemToModify.0.Transform.Rotation=&RigVMFunction_MathQuaternionFromEuler(Euler=(X=0,Y=0,Z=0)).Result
   )
   > n19 = &RigUnit_SphericalPoseReader(
    ActiveRegionScaleFactors=(PositiveWidth=0.5,NegativeWidth=0.5,PositiveHeight=1,NegativeHeight=1)
    ActiveRegionSize=0.1
    Debug=(bDrawDebug=false,bDraw2D=false,bDrawLocalAxes=false,DebugScale=25,DebugSegments=20,DebugThickness=0.25)
    DriverAxis=(X=-1,Y=0,Z=0)
    DriverItem=(Type=Bone,Name="upperarm_r")
    FalloffRegionScaleFactors=(PositiveWidth=0.25,NegativeWidth=0.25,PositiveHeight=1,NegativeHeight=1)
    FalloffSize=0.35
    OptionalParentItem=(Type=Bone,Name="clavicle_r")
    RotationOffset=(X=20,Y=45,Z=20)
   )
   n20 = RigVMVariableNode() {variable=Is Pistol Targeting Alpha}
   n21 = RigVMVariableNode() {variable=Is Rifle Targeting Alpha}
   n22 = &RigVMFunction_MathDoubleMul(
    B=&RigVMFunction_MathDoubleSub(B=&RigVMFunction_MathDoubleMul(A=@n19.OutputParam, B=@n20.Value).Result, A=@n19.OutputParam).Result
    A=&RigVMFunction_MathDoubleSub(B=&RigVMFunction_MathDoubleMul(A=@n19.OutputParam, B=@n21.Value).Result, A=@n19.OutputParam).Result
   )
   > &RigUnit_ModifyTransforms(
    ItemToModify=((Item=(Type=Bone,Name="dfrm_stomach_02_r"),Transform=(Rotation=(X=0,Y=0,Z=0,W=1),Translation=(X=-4,Y=0,Z=0),Scale3D=(X=1,Y=1,Z=1))))
    ItemToModify.0.Transform.Rotation=&RigVMFunction_MathQuaternionFromEuler(Euler=(X=0,Y=0,Z=0)).Result
    Weight=@n22.Result
   )
   > &RigUnit_ModifyTransforms(
    ItemToModify=((Item=(Type=Bone,Name="dfrm_stomach_05_r"),Transform=(Rotation=(X=0,Y=0,Z=0,W=1),Translation=(X=-8,Y=0,Z=0),Scale3D=(X=1,Y=1,Z=1))))
    ItemToModify.0.Transform.Rotation=&RigVMFunction_MathQuaternionFromEuler(Euler=(X=0,Y=0,Z=0)).Result
    Weight=@n22.Result
   )
   > &RigUnit_ModifyTransforms(
    ItemToModify=((Item=(Type=Bone,Name="dfrm_stomach_03_r"),Transform=(Rotation=(X=0,Y=0,Z=0,W=1),Translation=(X=-8,Y=0,Z=0),Scale3D=(X=1,Y=1,Z=1))))
    ItemToModify.0.Transform.Rotation=&RigVMFunction_MathQuaternionFromEuler(Euler=(X=0,Y=0,Z=0)).Result
    Weight=@n22.Result
   )
   > &RigUnit_ModifyTransforms(
    ItemToModify=((Item=(Type=Bone,Name="dfrm_stomach_06_r"),Transform=(Rotation=(X=0,Y=0,Z=0,W=1),Translation=(X=0,Y=0,Z=0),Scale3D=(X=1,Y=1,Z=1))))
    ItemToModify.0.Transform.Rotation=&RigVMFunction_MathQuaternionFromEuler(Euler=(X=0,Y=0,Z=0)).Result
    Weight=@n22.Result
   )
   > n23 = &RigUnit_SphericalPoseReader(
    ActiveRegionScaleFactors=(PositiveWidth=0.5,NegativeWidth=0.5,PositiveHeight=1,NegativeHeight=1)
    ActiveRegionSize=0.1
    Debug=(bDrawDebug=false,bDraw2D=false,bDrawLocalAxes=false,DebugScale=25,DebugSegments=20,DebugThickness=0.25)
    DriverAxis=(X=1,Y=0,Z=0)
    DriverItem=(Type=Bone,Name="upperarm_l")
    FalloffRegionScaleFactors=(PositiveWidth=0.25,NegativeWidth=0.25,PositiveHeight=1,NegativeHeight=1)
    FalloffSize=0.35
    OptionalParentItem=(Type=Bone,Name="clavicle_l")
    RotationOffset=(X=-20,Y=225,Z=20)
   )
   n24 = RigVMVariableNode() {variable=Is Rifle Targeting Alpha}
   n25 = RigVMVariableNode() {variable=Is Pistol Targeting Alpha}
   n26 = &RigVMFunction_MathDoubleMul(
    A=&RigVMFunction_MathDoubleSub(B=&RigVMFunction_MathDoubleMul(A=@n23.OutputParam, B=@n24.Value).Result, A=@n23.OutputParam).Result
    B=&RigVMFunction_MathDoubleSub(B=&RigVMFunction_MathDoubleMul(A=@n23.OutputParam, B=@n25.Value).Result, A=@n23.OutputParam).Result
   )
   > &RigUnit_ModifyTransforms(
    ItemToModify=((Item=(Type=Bone,Name="dfrm_stomach_02_l"),Transform=(Rotation=(X=0,Y=0,Z=0,W=1),Translation=(X=-4,Y=0,Z=0),Scale3D=(X=1,Y=1,Z=1))))
    ItemToModify.0.Transform.Rotation=&RigVMFunction_MathQuaternionFromEuler(Euler=(X=0,Y=0,Z=0)).Result
    Weight=@n26.Result
   )
   > &RigUnit_ModifyTransforms(
    ItemToModify=((Item=(Type=Bone,Name="dfrm_stomach_05_l"),Transform=(Rotation=(X=0,Y=0,Z=0,W=1),Translation=(X=-8,Y=0,Z=0),Scale3D=(X=1,Y=1,Z=1))))
    ItemToModify.0.Transform.Rotation=&RigVMFunction_MathQuaternionFromEuler(Euler=(X=0,Y=0,Z=0)).Result
    Weight=@n26.Result
   )
   > &RigUnit_ModifyTransforms(
    ItemToModify=((Item=(Type=Bone,Name="dfrm_stomach_03_l"),Transform=(Rotation=(X=0,Y=0,Z=0,W=1),Translation=(X=-8,Y=0,Z=0),Scale3D=(X=1,Y=1,Z=1))))
    ItemToModify.0.Transform.Rotation=&RigVMFunction_MathQuaternionFromEuler(Euler=(X=0,Y=0,Z=0)).Result
    Weight=@n26.Result
   )
   > &RigUnit_ModifyTransforms(
    ItemToModify=((Item=(Type=Bone,Name="dfrm_stomach_06_l"),Transform=(Rotation=(X=0,Y=0,Z=0,W=1),Translation=(X=0,Y=0,Z=0),Scale3D=(X=1,Y=1,Z=1))))
    ItemToModify.0.Transform.Rotation=&RigVMFunction_MathQuaternionFromEuler(Euler=(X=0,Y=0,Z=0)).Result
    Weight=@n26.Result
   )
   > n27 = &RigUnit_SphericalPoseReader(
    ActiveRegionScaleFactors=(PositiveWidth=1,NegativeWidth=1,PositiveHeight=1,NegativeHeight=1)
    ActiveRegionSize=0.1
    Debug=(bDrawDebug=false,bDraw2D=false,bDrawLocalAxes=false,DebugScale=25,DebugSegments=20,DebugThickness=0.25)
    DriverAxis=(X=-1,Y=0,Z=0)
    DriverItem=(Type=Bone,Name="clavicle_r")
    FalloffRegionScaleFactors=(PositiveWidth=1,NegativeWidth=1,PositiveHeight=0.5,NegativeHeight=1)
    FalloffSize=0.35
    OptionalParentItem=(Type=Bone,Name="spine_05")
    RotationOffset=(X=180,Y=-90,Z=-45)
   )
   n28 = &RigVMFunction_AlphaInterp(
    InterpSpeedDecreasing=8
    InterpSpeedIncreasing=10
    Scale=1
    bClampResult=False
    Value=&RigVMFunction_MathDoubleMul(A=@n27.OutputParam, B=RigVMVariableNode() {variable=Is Playing Melee DualWield Anim Alpha}.Value).Result
   )
   > &RigUnit_ModifyTransforms(
    ItemToModify=((Item=(Type=Bone,Name="attach_backpack"),Transform=(Rotation=(X=0,Y=0,Z=0,W=1),Translation=(X=3,Y=0,Z=3),Scale3D=(X=1,Y=1,Z=1))))
    Weight=@n28.Result
    ItemToModify.0.Transform.Rotation=&RigVMFunction_MathQuaternionFromEuler(Euler=(X=0,Y=0,Z=45)).Result
   )
   > n29 = &RigUnit_SphericalPoseReader(
    ActiveRegionScaleFactors=(PositiveWidth=1,NegativeWidth=1,PositiveHeight=1,NegativeHeight=1)
    ActiveRegionSize=0.1
    Debug=(bDrawDebug=false,bDraw2D=false,bDrawLocalAxes=false,DebugScale=25,DebugSegments=20,DebugThickness=0.25)
    DriverAxis=(X=1,Y=0,Z=0)
    DriverItem=(Type=Bone,Name="clavicle_l")
    FalloffRegionScaleFactors=(PositiveWidth=1,NegativeWidth=1,PositiveHeight=0.5,NegativeHeight=1)
    FalloffSize=0.35
    OptionalParentItem=(Type=Bone,Name="spine_05")
    RotationOffset=(X=0,Y=-90,Z=-45)
   )
   n30 = &RigVMFunction_AlphaInterp(
    InterpSpeedDecreasing=8
    InterpSpeedIncreasing=10
    Scale=1
    bClampResult=False
    Value=&RigVMFunction_MathDoubleMul(A=@n29.OutputParam, B=RigVMVariableNode() {variable=Is Playing Melee DualWield Anim Alpha}.Value).Result
   )
   > &RigUnit_ModifyTransforms(
    ItemToModify=((Item=(Type=Bone,Name="attach_backpack"),Transform=(Rotation=(X=0,Y=0,Z=0,W=1),Translation=(X=-3,Y=0,Z=3),Scale3D=(X=1,Y=1,Z=1))))
    Weight=@n30.Result
    ItemToModify.0.Transform.Rotation=&RigVMFunction_MathQuaternionFromEuler(Euler=(X=0,Y=0,Z=-45)).Result
   )
   > n31 = &RigUnit_SphericalPoseReader(
    ActiveRegionScaleFactors=(PositiveWidth=1,NegativeWidth=1,PositiveHeight=1,NegativeHeight=1)
    ActiveRegionSize=0.1
    Debug=(bDrawDebug=false,bDraw2D=false,bDrawLocalAxes=false,DebugScale=25,DebugSegments=20,DebugThickness=0.25)
    DriverAxis=(X=-1,Y=0,Z=0)
    DriverItem=(Type=Bone,Name="dfrm_clavicle_r")
    FalloffRegionScaleFactors=(PositiveWidth=1,NegativeWidth=1,PositiveHeight=0.5,NegativeHeight=1)
    FalloffSize=0.35
    OptionalParentItem=(Type=Bone,Name="spine_05")
    RotationOffset=(X=180,Y=-90,Z=-45)
   )
   n32 = &RigVMFunction_AlphaInterp(
    InterpSpeedDecreasing=8
    InterpSpeedIncreasing=10
    Scale=1
    bClampResult=False
    Value=&RigVMFunction_MathDoubleMul(A=@n31.OutputParam, B=RigVMVariableNode() {variable=Is Melee OneHanded Alpha}.Value).Result
   )
   > &RigUnit_ModifyTransforms(
    ItemToModify=((Item=(Type=Bone,Name="attach_backpack"),Transform=(Rotation=(X=0,Y=0,Z=0,W=1),Translation=(X=5,Y=0,Z=1),Scale3D=(X=1,Y=1,Z=1))))
    Weight=@n32.Result
    ItemToModify.0.Transform.Rotation=&RigVMFunction_MathQuaternionFromEuler(Euler=(X=0,Y=0,Z=15)).Result
   )
   > n33 = &RigUnit_SphericalPoseReader(
    ActiveRegionScaleFactors=(PositiveWidth=1,NegativeWidth=1,PositiveHeight=1,NegativeHeight=1)
    ActiveRegionSize=0.1
    Debug=(bDrawDebug=false,bDraw2D=false,bDrawLocalAxes=false,DebugScale=25,DebugSegments=20,DebugThickness=0.25)
    DriverAxis=(X=1,Y=0,Z=0)
    DriverItem=(Type=Bone,Name="dfrm_clavicle_l")
    FalloffRegionScaleFactors=(PositiveWidth=1,NegativeWidth=1,PositiveHeight=0.5,NegativeHeight=1)
    FalloffSize=0.35
    OptionalParentItem=(Type=Bone,Name="spine_05")
    RotationOffset=(X=0,Y=-90,Z=-45)
   )
   n34 = &RigVMFunction_AlphaInterp(
    InterpSpeedDecreasing=8
    InterpSpeedIncreasing=10
    Scale=1
    bClampResult=False
    Value=&RigVMFunction_MathDoubleMul(A=@n33.OutputParam, B=RigVMVariableNode() {variable=Is Melee OneHanded Alpha}.Value).Result
   )
   > &RigUnit_ModifyTransforms(
    ItemToModify=((Item=(Type=Bone,Name="attach_backpack"),Transform=(Rotation=(X=0,Y=0,Z=0,W=1),Translation=(X=-5,Y=0,Z=1),Scale3D=(X=1,Y=1,Z=1))))
    Weight=@n34.Result
    ItemToModify.0.Transform.Rotation=&RigVMFunction_MathQuaternionFromEuler(Euler=(X=0,Y=0,Z=-15)).Result
   )
   Remap = &RigVMFunction_MathDoubleRemap(
    SourceMaximum=40
    SourceMinimum=25
    TargetMaximum=1
    TargetMinimum=0
    Value=&RigVMFunction_MathVectorDistance(
     A=&RigUnit_GetTransform(Item=(Type=Bone,Name="pelvis"), bInitial=False).Transform.Translation
     B=&RigUnit_GetTransform(Item=(Type=Bone,Name="lowerarm_l"), bInitial=False).Transform.Translation
    ).Result
   )
   Multiply = &RigVMFunction_MathDoubleMul(B=&RigVMFunction_MathDoubleSub(A=1, B=@Remap.Result).Result, A=RigVMVariableNode() {variable=SPR_upperarm_l_down_A_alpha}.Value)
   RigVMVariableNode(Value=@Multiply.Result, ExecuteContext=@SequenceAggregate_1.B) {variable=mult_spr_dst_lowerarm_r_pelvis_alpha}
   RigVMFunction_MathVectorDistance = &RigVMFunction_MathVectorDistance(
    A=&RigUnit_GetTransform(Item=(Type=Bone,Name="hand_r"), bInitial=False).Transform.Translation
    B=&RigUnit_GetTransform(Item=(Type=Bone,Name="thigh_twist_01_r"), bInitial=False).Transform.Translation
   )
   > RigVMVariableNode(Value=&RigVMFunction_MathDoubleSub(
     A=1
     B=&RigVMFunction_MathDoubleRemap(SourceMaximum=59, SourceMinimum=19.5, TargetMaximum=1, TargetMinimum=0, Value=@RigVMFunction_MathVectorDistance.Result).Result
    ).Result) {variable=mult_spr_dst_hand_r_thigh_twist_r_alpha}
   n35 = &RigVMFunction_MathQuaternionSwingTwist(
    TwistAxis=(X=1,Y=0,Z=0)
    Input=&RigVMFunction_MathTransformMakeRelative(
     Parent=&RigUnit_GetTransform(Item=(Type=Bone,Name="hand_l"), Space=LocalSpace, bInitial=true).Transform
     Global=&RigUnit_GetTransform(Item=(Type=Bone,Name="hand_l"), Space=LocalSpace).Transform
    ).Local.Rotation
   )
   &RigUnit_SetRotation(
    Item=(Type=Bone,Name="dfrm_lowerarm_twist_01_l")
    Weight=0.75
    bPropagateToChildren=false
    Value=@n35.Twist
    ExecutePin=@SequenceAggregate_1.E
   )
   > &RigUnit_SetRotation(Item=(Type=Bone,Name="dfrm_lowerarm_twist_02_l"), Weight=0.5, bPropagateToChildren=false, Value=@n35.Twist)
   n36 = &RigVMFunction_MathQuaternionSwingTwist(
    TwistAxis=(X=1,Y=0,Z=0)
    Input=&RigVMFunction_MathTransformMakeRelative(
     Parent=&RigUnit_GetTransform(Item=(Type=Bone,Name="hand_r"), Space=LocalSpace, bInitial=true).Transform
     Global=&RigUnit_GetTransform(Item=(Type=Bone,Name="hand_r"), Space=LocalSpace).Transform
    ).Local.Rotation
   )
   > &RigUnit_SetRotation(Item=(Type=Bone,Name="dfrm_lowerarm_twist_01_r"), Weight=0.75, bPropagateToChildren=false, Value=@n36.Twist)
   > &RigUnit_SetRotation(Item=(Type=Bone,Name="dfrm_lowerarm_twist_02_r"), Weight=0.5, bPropagateToChildren=false, Value=@n36.Twist)
   RigVMFunction_Sequence_1 = RigVMAggregateNode(ExecuteContext=@SequenceAggregate_1.F)
   n37 = &RigVMFunction_MathDoubleMul(
    A=RigVMVariableNode() {variable=SPR_upperarm_r_down_A_alpha}.Value
    B=RigVMVariableNode() {variable=mult_spr_dst_hand_r_thigh_twist_r_alpha}.Value
   )
   &RigUnit_ModifyTransforms(
    ItemToModify="((Item=(Type=Bone,Name=\"dfrm_stomach_01_r\"),Transform=(Rotation=(X=0,Y=0,Z=0,W=1),Translation=(X=-2,Y=0,Z=0),Scale3D=(X=1,Y=1,Z=1))),(Item=(Type=Bone,Name=\"dfrm_stomach_02_r\"),Transform=(Rotation=(X=0,Y=0,Z=0,W=1),Translation=(X=-8,Y=3,Z=0),Scale3D=(X=1,Y= ...(truncated)"
    WeightMinimum=0
    Weight=@n37.Result
    ExecutePin=@RigVMFunction_Sequence_1.A
   )
   > &RigUnit_ModifyTransforms(
    ItemToModify="((Item=(Type=Bone,Name=\"dfrm_stomach_06_r\"),Transform=(Rotation=(X=0,Y=0,Z=0,W=1),Translation=(X=-10,Y=2,Z=0),Scale3D=(X=1,Y=1,Z=1))),(Item=(Type=Bone,Name=\"dfrm_stomach_05_r\"),Transform=(Rotation=(X=0,Y=0,Z=0,W=1),Translation=(X=-8,Y=0,Z=0),Scale3D=(X=1,Y ...(truncated)"
    WeightMinimum=0
    Weight=@n37.Result
   )
   RigVMFunction_MathQuaternionFromEuler = &RigVMFunction_MathQuaternionFromEuler(Euler=(X=-45,Y=0,Z=0))
   > &RigUnit_ModifyTransforms(
    ItemToModify=((Item=(Type=Bone,Name="dfrm_lowerarm_twist_01_a_l"),Transform=(Rotation=(X=0,Y=0,Z=0,W=1),Translation=(X=0,Y=0,Z=0),Scale3D=(X=1,Y=1,Z=1))))
    WeightMinimum=0
    ItemToModify.0.Transform.Rotation=@RigVMFunction_MathQuaternionFromEuler.Result
    Weight=@n37.Result
   )
   RigVMFunction_Sequence = &RigVMFunction_Sequence(ExecuteContext=@RigVMFunction_Sequence_1.B)
   n38 = RigVMVariableNode() {variable=mult_spr_dst_lowerarm_r_pelvis_alpha}
   n39 = &RigVMFunction_MathDoubleRemap(SourceMaximum=1, SourceMinimum=0.5, TargetMaximum=1, TargetMinimum=0, Value=@n38.Value)
   &RigUnit_ModifyTransforms(
    ItemToModify=((Item=(Type=Bone,Name="dfrm_spine_01_l"),Transform=(Rotation=(X=0,Y=0,Z=0,W=1),Translation=(X=0,Y=0,Z=0),Scale3D=(X=1,Y=1,Z=-0.5))))
    WeightMinimum=0
    ExecutePin=@RigVMFunction_Sequence.A
    Weight=@n39.Result
   )
   > &RigUnit_ModifyTransforms(
    ItemToModify=((Item=(Type=Bone,Name="dfrm_spine_01_front"),Transform=(Rotation=(X=0,Y=0,Z=0,W=1),Translation=(X=0,Y=0,Z=0),Scale3D=(X=1,Y=-0.5,Z=0))))
    WeightMinimum=0
    Weight=@n39.Result
   )
   > &RigUnit_ModifyTransforms(
    ItemToModify=((Item=(Type=Bone,Name="dfrm_spine_01_back"),Transform=(Rotation=(X=0,Y=0,Z=0,W=1),Translation=(X=0,Y=0,Z=0),Scale3D=(X=1,Y=-0.5,Z=0))))
    WeightMinimum=0
    Weight=@n39.Result
   )
   &RigUnit_ModifyTransforms(
    ItemToModify="((Item=(Type=Bone,Name=\"dfrm_stomach_mid\"),Transform=(Rotation=(X=0,Y=0,Z=0,W=1),Translation=(X=-5,Y=0,Z=0),Scale3D=(X=0.5,Y=1,Z=1))),(Item=(Type=Bone,Name=\"dfrm_stomach_lower_mid\"),Transform=(Rotation=(X=0,Y=0,Z=0,W=1),Translation=(X=-2.5,Y=0,Z=0),Scale3D ...(truncated)"
    WeightMinimum=0
    ExecutePin=@RigVMFunction_Sequence.B
    Weight=@n38.Value
   )
   > &RigUnit_ModifyTransforms(
    ItemToModify="((Item=(Type=Bone,Name=\"dfrm_stomach_01_l\"),Transform=(Rotation=(X=0,Y=0,Z=0,W=1),Translation=(X=-5,Y=0,Z=0),Scale3D=(X=0.75,Y=1,Z=1))),(Item=(Type=Bone,Name=\"dfrm_stomach_04_l\"),Transform=(Rotation=(X=0,Y=0,Z=0,W=1),Translation=(X=0,Y=0,Z=0),Scale3D=(X=1, ...(truncated)"
    WeightMinimum=0
    Weight=@n38.Value
   )
   > &RigUnit_ModifyTransforms(
    ItemToModify="((Item=(Type=Bone,Name=\"dfrm_stomach_02_l\"),Transform=(Rotation=(X=0,Y=0,Z=0,W=1),Translation=(X=-3,Y=0,Z=0),Scale3D=(X=0.85,Y=1,Z=1))),(Item=(Type=Bone,Name=\"dfrm_stomach_05_l\"),Transform=(Rotation=(X=0,Y=0,Z=0,W=1),Translation=(X=0,Y=0,Z=0),Scale3D=(X=1, ...(truncated)"
    WeightMinimum=0
    Weight=@n38.Value
   )
   ClampSpatially = /Script/RigVM.RigVMFunction_MathTransformClampSpatially(
    Axis=X
    DebugColor=(R=1,G=0,B=0,A=1)
    DebugThickness=1
    Maximum=120
    Minimum=30
    Type=Capsule
    bDrawDebug=false
    Value=&RigUnit_GetTransform(Item=(Type=Bone,Name="dfrm_upperarm_clamp_r_"), bInitial=False).Transform
    Space=&RigUnit_GetTransform(Item=(Type=Control,Name="dfrm_pelvis_ctrl_3"), bInitial=False).Transform
   ) {
    is=struct RigVMUnitNode
    template="ClampSpatially::Execute(in Value,in Axis,in Type,in Minimum,in Maximum,in Space,in bDrawDebug,in DebugColor,in DebugThickness,out Result)"
   }
   &RigUnit_SetTransform(
    Item=(Type=Bone,Name="dfrm_upperarm_clamp_r_")
    Weight=1
    bInitial=False
    Value=@ClampSpatially.Result
    ExecutePin=@RigVMFunction_Sequence_1.C
   )
   > &RigUnit_SetTransform(
    Item=(Type=Bone,Name="dfrm_lowerarm_r")
    Weight=1
    bInitial=False
    Value=&RigVMFunction_MathTransformClampSpatially(
     Axis=X
     DebugColor=(R=1,G=0,B=0,A=1)
     DebugThickness=1
     Maximum=75
     Minimum=40
     Type=Capsule
     bDrawDebug=false
     Value=&RigUnit_GetTransform(Item=(Type=Bone,Name="dfrm_lowerarm_r"), bInitial=False).Transform
     Space=&RigUnit_GetTransform(Item=(Type=Control,Name="dfrm_pelvis_ctrl"), bInitial=False).Transform
    ) {template="ClampSpatially::Execute(in Value,in Axis,in Type,in Minimum,in Maximum,in Space,in bDrawDebug,in DebugColor,in DebugThickness,out Result)"}.Result
   )
   > &RigUnit_SetTransform(
    Item=(Type=Bone,Name="dfrm_lowerarm_twist_02_r")
    Weight=1
    bInitial=False
    Value=&RigVMFunction_MathTransformClampSpatially(
     Axis=Y
     DebugColor=(R=1,G=0,B=0,A=1)
     DebugThickness=1
     Maximum=10000
     Minimum=40
     Type=Sphere
     bDrawDebug=false
     Value=&RigUnit_GetTransform(Item=(Type=Bone,Name="dfrm_lowerarm_twist_02_r"), bInitial=False).Transform
     Space=&RigUnit_GetTransform(Item=(Type=Control,Name="dfrm_pelvis_ctrl"), bInitial=False).Transform
    ) {template="ClampSpatially::Execute(in Value,in Axis,in Type,in Minimum,in Maximum,in Space,in bDrawDebug,in DebugColor,in DebugThickness,out Result)"}.Result
   )
   > &RigUnit_SetTransform(
    Item=(Type=Bone,Name="dfrm_lowerarm_twist_01_r")
    Weight=1
    bInitial=False
    Value=&RigVMFunction_MathTransformClampSpatially(
     Axis=Y
     DebugColor=(R=1,G=0,B=0,A=1)
     DebugThickness=1
     Maximum=1000
     Minimum=40
     Type=Sphere
     bDrawDebug=false
     Value=&RigUnit_GetTransform(Item=(Type=Bone,Name="dfrm_lowerarm_twist_01_r"), bInitial=False).Transform
     Space=&RigUnit_GetTransform(Item=(Type=Control,Name="dfrm_pelvis_ctrl"), bInitial=False).Transform
    ) {template="ClampSpatially::Execute(in Value,in Axis,in Type,in Minimum,in Maximum,in Space,in bDrawDebug,in DebugColor,in DebugThickness,out Result)"}.Result
   )
   n40 = &RigUnit_SphericalPoseReader(
    ActiveRegionScaleFactors=(PositiveWidth=0.5,NegativeWidth=0.5,PositiveHeight=0.5,NegativeHeight=0.5)
    ActiveRegionSize=0.1
    Debug=(bDrawDebug=false,bDraw2D=false,bDrawLocalAxes=false,DebugScale=25,DebugSegments=20,DebugThickness=0.25)
    DriverAxis=(X=-1,Y=0,Z=0)
    DriverItem=(Type=Bone,Name="upperarm_r")
    ExecutePin=()
    FalloffRegionScaleFactors=(PositiveWidth=1,NegativeWidth=1,PositiveHeight=1,NegativeHeight=1)
    FalloffSize=0.25
    OptionalParentItem=(Type=Bone,Name="clavicle_r")
    RotationOffset=(X=-22,Y=27,Z=0)
   )
   n41 = &RigVMFunction_AlphaInterp(
    InterpSpeedDecreasing=7
    InterpSpeedIncreasing=10
    Scale=1
    bClampResult=False
    Value=&RigVMFunction_MathDoubleMul(B=@n40.OutputParam, A=@n7.Value).Result
   )
   > &RigUnit_ModifyTransforms(
    ItemToModify=((Item=(Type=Bone,Name="dfrm_stomach_mid"),Transform=(Rotation=(X=0,Y=0,Z=0,W=1),Translation=(X=-15,Y=0,Z=0),Scale3D=(X=1,Y=1,Z=1))))
    Weight=@n41.Result
    ItemToModify.0.Transform.Rotation=&RigVMFunction_MathQuaternionFromEuler(Euler=(X=0,Y=0,Z=0)).Result
   )
   > &RigUnit_ModifyTransforms(
    ItemToModify=((Item=(Type=Bone,Name="dfrm_stomach_lower_mid"),Transform=(Rotation=(X=0,Y=0,Z=0,W=1),Translation=(X=-15,Y=0,Z=0),Scale3D=(X=1,Y=1,Z=1))))
    Weight=@n41.Result
    ItemToModify.0.Transform.Rotation=&RigVMFunction_MathQuaternionFromEuler(Euler=(X=0,Y=0,Z=0)).Result
   )
   > &RigUnit_ModifyTransforms(
    ItemToModify=((Item=(Type=Bone,Name="dfrm_stomach_01_r"),Transform=(Rotation=(X=0,Y=0,Z=0,W=1),Translation=(X=-8,Y=0,Z=0),Scale3D=(X=1,Y=1,Z=1))))
    Weight=@n41.Result
    ItemToModify.0.Transform.Rotation=&RigVMFunction_MathQuaternionFromEuler(Euler=(X=0,Y=0,Z=0)).Result
   )
   > &RigUnit_ModifyTransforms(
    ItemToModify=((Item=(Type=Bone,Name="dfrm_stomach_04_r"),Transform=(Rotation=(X=0,Y=0,Z=0,W=1),Translation=(X=-8,Y=0,Z=0),Scale3D=(X=1,Y=1,Z=1))))
    Weight=@n41.Result
    ItemToModify.0.Transform.Rotation=&RigVMFunction_MathQuaternionFromEuler(Euler=(X=0,Y=0,Z=0)).Result
   )
   n42 = &RigUnit_SphericalPoseReader(
    ActiveRegionScaleFactors=(PositiveWidth=0.5,NegativeWidth=0.5,PositiveHeight=2,NegativeHeight=0)
    ActiveRegionSize=0.1
    Debug=(bDrawDebug=false,bDraw2D=false,bDrawLocalAxes=false,DebugScale=25,DebugSegments=20,DebugThickness=0.25)
    DriverAxis=(X=-1,Y=0,Z=0)
    DriverItem=(Type=Bone,Name="upperarm_r")
    ExecutePin=()
    FalloffRegionScaleFactors=(PositiveWidth=0.25,NegativeWidth=0.25,PositiveHeight=0.5,NegativeHeight=0.5)
    FalloffSize=0.45
    OptionalParentItem=(Type=Bone,Name="clavicle_r")
    RotationOffset=(X=-65,Y=20,Z=15)
   )
   n43 = &RigVMFunction_AlphaInterp(
    InterpSpeedDecreasing=7
    InterpSpeedIncreasing=10
    Scale=1
    bClampResult=False
    Value=&RigVMFunction_MathDoubleMul(B=@n42.OutputParam, A=@n9.Value).Result
   )
   > &RigUnit_ModifyTransforms(
    ItemToModify=((Item=(Type=Bone,Name="dfrm_stomach_mid"),Transform=(Rotation=(X=0,Y=0,Z=0,W=1),Translation=(X=-15,Y=0,Z=0),Scale3D=(X=1,Y=1,Z=1))))
    Weight=@n43.Result
    ItemToModify.0.Transform.Rotation=&RigVMFunction_MathQuaternionFromEuler(Euler=(X=0,Y=0,Z=0)).Result
   )
   > &RigUnit_ModifyTransforms(
    ItemToModify=((Item=(Type=Bone,Name="dfrm_stomach_lower_mid"),Transform=(Rotation=(X=0,Y=0,Z=0,W=1),Translation=(X=-15,Y=0,Z=0),Scale3D=(X=1,Y=1,Z=1))))
    Weight=@n43.Result
    ItemToModify.0.Transform.Rotation=&RigVMFunction_MathQuaternionFromEuler(Euler=(X=0,Y=0,Z=0)).Result
   )
   > &RigUnit_ModifyTransforms(
    ItemToModify=((Item=(Type=Bone,Name="dfrm_stomach_01_r"),Transform=(Rotation=(X=0,Y=0,Z=0,W=1),Translation=(X=-8,Y=0,Z=0),Scale3D=(X=1,Y=1,Z=1))))
    Weight=@n43.Result
    ItemToModify.0.Transform.Rotation=&RigVMFunction_MathQuaternionFromEuler(Euler=(X=0,Y=0,Z=0)).Result
   )
   > &RigUnit_ModifyTransforms(
    ItemToModify=((Item=(Type=Bone,Name="dfrm_stomach_04_r"),Transform=(Rotation=(X=0,Y=0,Z=0,W=1),Translation=(X=-8,Y=0,Z=0),Scale3D=(X=1,Y=1,Z=1))))
    Weight=@n43.Result
    ItemToModify.0.Transform.Rotation=&RigVMFunction_MathQuaternionFromEuler(Euler=(X=0,Y=0,Z=0)).Result
   )
   &RigUnit_ModifyTransforms(
    ExecutePin=()
    ItemToModify=((Item=(Type=Bone,Name="dfrm_pelvis"),Transform=(Rotation=(X=0,Y=0,Z=-0.3007,W=0.9537),Translation=(X=-3,Y=-8.5,Z=0),Scale3D=(X=1,Y=1,Z=1))))
    WeightMinimum=0
    Weight=&RigVMFunction_MathDoubleMul(B=0, A=RigVMVariableNode() {variable=mult_spr_dst_lowerarm_r_pelvis_alpha}.Value).Result
   )
   RigVMRerouteNode(Value=@n38.Value)
   RigVMRerouteNode(Value=@RigVMFunction_Sequence_1.D)
  ]}
 SequenceAggregate_1_1_ContainedGraph {nodes [
   n1 = RigVMFunctionEntryNode()
   SequenceAggregate_1 = &RigVMFunction_Sequence(ExecuteContext=@n1.ExecuteContext)
   SequenceAggregate_1_1 = &RigVMFunction_Sequence(ExecuteContext=@SequenceAggregate_1.B)
   SequenceAggregate_1_2 = &RigVMFunction_Sequence(ExecuteContext=@SequenceAggregate_1_1.B)
   SequenceAggregate_1_3 = &RigVMFunction_Sequence(ExecuteContext=@SequenceAggregate_1_2.B)
   SequenceAggregate_1_4 = &RigVMFunction_Sequence(ExecuteContext=@SequenceAggregate_1_3.B)
   SequenceAggregate_1_5 = &RigVMFunction_Sequence(ExecuteContext=@SequenceAggregate_1_4.B)
   RigVMFunctionReturnNode(
    A=@SequenceAggregate_1.A
    B=@SequenceAggregate_1_1.A
    C=@SequenceAggregate_1_2.A
    D=@SequenceAggregate_1_3.A
    E=@SequenceAggregate_1_4.A
    F=@SequenceAggregate_1_5.A
    G=@SequenceAggregate_1_5.B
   )
  ]}
]
hierarchy [
 root {loc=<0 0 0>, rot=<0 0 0 1>}
 pelvis {parent=root, loc=<0 -0.7321 92.4218>, rot=<0.0127 0.707 -0.0127 -0.707>}
 spine_01 {parent=pelvis, loc=<5.6002 -0.5128 0>, rot=<0 0 0.034 -0.9994>}
 spine_02 {parent=spine_01, loc=<8.0753 0.0139 0>, rot=<0 0 0.0238 -0.9997>}
 spine_03 {parent=spine_02, loc=<7.0393 0 0>, rot=<0 0 0.0149 -0.9999>}
 spine_04 {parent=spine_03, loc=<8.1024 0 0>, rot=<0 0 -0.0132 -0.9999>}
 spine_05 {parent=spine_04, loc=<10.0146 0 0>, rot=<0 0 -0.0015 -1>}
 clavicle_l {parent=spine_05, loc=<2.5856 -0.2866 -4.3777>, rot=<0.0465 -0.7289 -0.009 -0.6829>}
 upperarm_l {parent=clavicle_l, loc=<18.2264 0.1342 -0.001>, rot=<0.0253 -0.2693 0.0013 -0.9627>}
 lowerarm_l {parent=upperarm_l, loc=<29.0071 0.1686 -1.239>, rot=<0.0391 -0.0099 -0.2453 -0.9686>}
 hand_l {parent=lowerarm_l, loc=<25.8607 0.222 1.0274>, rot=<-0.7193 -0.0988 0.1193 -0.6773>}
 index_metacarpal_l {parent=hand_l, loc=<1.6535 0.0183 -2.96>, rot=<0.0158 -0.0124 -0.0173 -0.9996>}
 index_01_l {parent=index_metacarpal_l, loc=<8.3008 0.0357 -0.0524>, rot=<-0.0044 0.0485 0.0924 -0.9945>}
 index_02_l {parent=index_01_l, loc=<3.3587 -0.3004 0.0452>, rot=<-0.002 0.0027 0.3225 -0.9466>}
 index_03_l {parent=index_02_l, loc=<2.5578 -0.123 -0.0252>, rot=<-0.002 0.0067 0.2558 -0.9667>}
 middle_metacarpal_l {parent=hand_l, loc=<2.1856 -0.1725 -0.5984>, rot=<0.006 0.0168 -0.0329 -0.9993>}
 middle_01_l {parent=middle_metacarpal_l, loc=<7.3275 -0.2447 0.0129>, rot=<-0.0208 0.0042 0.0858 -0.9961>}
 middle_02_l {parent=middle_01_l, loc=<4.0123 0.1401 0.1175>, rot=<-0.0218 0.0258 0.3512 -0.9357>}
 middle_03_l {parent=middle_02_l, loc=<2.6077 0.0534 0.1373>, rot=<0.0082 0.0188 0.1708 -0.9851>}
 pinky_metacarpal_l {parent=hand_l, loc=<1.7616 -0.9753 3.3814>, rot=<-0.0397 0.0814 -0.069 -0.9935>}
 pinky_01_l {parent=pinky_metacarpal_l, loc=<7.1709 -0.4831 -0.1094>, rot=<-0.0654 -0.0035 0.2242 -0.9723>}
 pinky_02_l {parent=pinky_01_l, loc=<2.8274 0.0893 -0.1043>, rot=<0.0037 0.0011 0.2854 -0.9584>}
 pinky_03_l {parent=pinky_02_l, loc=<2.2046 0.0307 -0.0078>, rot=<-0.0002 0.0098 0.1445 -0.9895>}
 ring_metacarpal_l {parent=hand_l, loc=<2.2074 -0.3625 1.3638>, rot=<-0.0284 0.0722 -0.0458 -0.9959>}
 ring_01_l {parent=ring_metacarpal_l, loc=<7.0295 -0.6197 -0.1446>, rot=<-0.0489 -0.0252 0.1892 -0.9804>}
 ring_02_l {parent=ring_01_l, loc=<3.6174 0.4647 -0.1027>, rot=<0 0 0.2806 -0.9598>}
 ring_03_l {parent=ring_02_l, loc=<2.4152 0.114 0.0199>, rot=<0 0 0.1189 -0.9929>}
 thumb_01_l {parent=hand_l, loc=<1.3118 -0.3191 -2.459>, rot=<0.6546 -0.4042 0.0664 -0.6354>}
 thumb_02_l {parent=thumb_01_l, loc=<5.9336 -0.0448 0.0401>, rot=<-0.0037 0.0213 0.298 -0.9543>}
 thumb_03_l {parent=thumb_02_l, loc=<2.2223 0.3682 -0.0907>, rot=<0.0118 -0.013 0.0066 -0.9998>}
 ArmFeather_02_l_jnt {parent=lowerarm_l, loc=<2.0507 -13.2978 -0.2203>, rot=<-0.0542 -0.0516 0.3423 -0.9366>}
 ArmFeather_03_l_jnt {parent=lowerarm_l, loc=<14.375 -9.4807 -0.8462>, rot=<-0.0558 -0.0194 0.1991 -0.9782>}
 ArmFeather_01_l_jnt {parent=upperarm_l, loc=<21.7807 -14.5023 -2.5933>, rot=<-0.062 0.0548 0.2181 -0.9724>}
 upperarm_twist_01_l {parent=upperarm_l, loc=<7.9796 0.0843 -0.6195>, rot=<0 0 0 -1>}
 deform_upperarm_bk_base_l {parent=upperarm_twist_01_l, loc=<-6.8983 -13.3826 -1.3258>, rot=<0 0 0 1>}
 deform_upperarm_bk_l {parent=deform_upperarm_bk_base_l, loc=<0 0 0>, rot=<0 0 0 1>}
 deform_upperarm_fr_base_l {parent=upperarm_twist_01_l, loc=<-8.3487 10.7117 -2.1455>, rot=<0 0 0 1>}
 deform_upperarm_fr_l {parent=deform_upperarm_fr_base_l, loc=<0 0 0>, rot=<0 0 0 1>}
 deform_upperarm_up_base_l {parent=upperarm_twist_01_l, loc=<-11.2012 -1.2234 8.7784>, rot=<0 0 0 1>}
 deform_upperarm_up_l {parent=deform_upperarm_up_base_l, loc=<0 0 0>, rot=<0 0 0 1>}
 clavicle_r {parent=spine_05, loc=<2.5856 -0.2865 4.3777>, rot=<0.7289 0.0465 -0.6829 0.009>}
 upperarm_r {parent=clavicle_r, loc=<-18.2264 -0.1342 0.001>, rot=<0.0253 -0.2693 0.0013 -0.9627>}
 lowerarm_r {parent=upperarm_r, loc=<-29.0072 -0.1686 1.239>, rot=<0.0391 -0.0099 -0.2453 -0.9686>}
 hand_r {parent=lowerarm_r, loc=<-25.8607 -0.222 -1.0274>, rot=<-0.7193 -0.0988 0.1193 -0.6773>}
 index_metacarpal_r {parent=hand_r, loc=<-1.6535 -0.0183 2.96>, rot=<0.0158 -0.0124 -0.0173 -0.9996>}
 index_01_r {parent=index_metacarpal_r, loc=<-8.3008 -0.0357 0.0524>, rot=<-0.0044 0.0485 0.0924 -0.9945>}
 index_02_r {parent=index_01_r, loc=<-3.3587 0.3004 -0.0452>, rot=<-0.002 0.0027 0.3225 -0.9466>}
 index_03_r {parent=index_02_r, loc=<-2.5578 0.123 0.0252>, rot=<-0.002 0.0067 0.2558 -0.9667>}
 middle_metacarpal_r {parent=hand_r, loc=<-2.1856 0.1725 0.5984>, rot=<0.006 0.0168 -0.0329 -0.9993>}
 middle_01_r {parent=middle_metacarpal_r, loc=<-7.3275 0.2447 -0.0129>, rot=<-0.0208 0.0042 0.0858 -0.9961>}
 middle_02_r {parent=middle_01_r, loc=<-4.0123 -0.1401 -0.1175>, rot=<-0.0218 0.0258 0.3512 -0.9357>}
 middle_03_r {parent=middle_02_r, loc=<-2.6077 -0.0533 -0.1373>, rot=<0.0082 0.0188 0.1708 -0.9851>}
 pinky_metacarpal_r {parent=hand_r, loc=<-1.7616 0.9753 -3.3814>, rot=<-0.0397 0.0814 -0.069 -0.9935>}
 pinky_01_r {parent=pinky_metacarpal_r, loc=<-7.1709 0.4831 0.1094>, rot=<-0.0654 -0.0035 0.2242 -0.9723>}
 pinky_02_r {parent=pinky_01_r, loc=<-2.8274 -0.0892 0.1043>, rot=<0.0037 0.0011 0.2854 -0.9584>}
 pinky_03_r {parent=pinky_02_r, loc=<-2.2046 -0.0307 0.0078>, rot=<-0.0002 0.0098 0.1445 -0.9895>}
 ring_metacarpal_r {parent=hand_r, loc=<-2.2074 0.3625 -1.3638>, rot=<-0.0284 0.0722 -0.0458 -0.9959>}
 ring_01_r {parent=ring_metacarpal_r, loc=<-7.0295 0.6197 0.1446>, rot=<-0.0489 -0.0252 0.1892 -0.9804>}
 ring_02_r {parent=ring_01_r, loc=<-3.6174 -0.4647 0.1027>, rot=<0 0 0.2806 -0.9598>}
 ring_03_r {parent=ring_02_r, loc=<-2.4153 -0.1141 -0.0199>, rot=<0 0 0.1189 -0.9929>}
 thumb_01_r {parent=hand_r, loc=<-1.3118 0.3191 2.459>, rot=<0.6546 -0.4042 0.0664 -0.6354>}
 thumb_02_r {parent=thumb_01_r, loc=<-5.9335 0.0447 -0.0401>, rot=<-0.0037 0.0213 0.298 -0.9543>}
 thumb_03_r {parent=thumb_02_r, loc=<-2.2223 -0.3682 0.0907>, rot=<0.0118 -0.013 0.0066 -0.9998>}
 ArmFeather_02_r_jnt {parent=lowerarm_r, loc=<-2.0509 13.2979 0.2207>, rot=<-0.3423 -0.9366 -0.0542 0.0516>}
 ArmFeather_03_r_jnt {parent=lowerarm_r, loc=<-14.3749 9.4806 0.8461>, rot=<-0.1991 -0.9782 -0.0558 0.0194>}
 ArmFeather_01_r_jnt {parent=upperarm_r, loc=<-21.7806 14.5023 2.593>, rot=<-0.2181 -0.9724 -0.062 -0.0548>}
 upperarm_twist_01_r {parent=upperarm_r, loc=<-7.9796 -0.0843 0.6195>, rot=<0 0 0 -1>}
 deform_upperarm_bk_base_r {parent=upperarm_twist_01_r, loc=<6.8985 13.3826 1.3257>, rot=<0 0 0 1>}
 deform_upperarm_bk_r {parent=deform_upperarm_bk_base_r, loc=<0 0 0>, rot=<0 0 0 1>}
 deform_upperarm_fr_base_r {parent=upperarm_twist_01_r, loc=<8.3485 -10.7117 2.1456>, rot=<0 0 0 1>}
 deform_upperarm_fr_r {parent=deform_upperarm_fr_base_r, loc=<0 0 0>, rot=<0 0 0 1>}
 deform_upperarm_up_base_r {parent=upperarm_twist_01_r, loc=<11.2015 1.2234 -8.7788>, rot=<0 0 0 1>}
 deform_upperarm_up_r {parent=deform_upperarm_up_base_r, loc=<0 0 0>, rot=<0 0 0 1>}
 neck_01 {parent=spine_05, loc=<10.815 1.9196 0>, rot=<0 0 -0.1864 -0.9825>}
 neck_02 {parent=neck_01, loc=<4.7347 0.3045 0>, rot=<0 0 0.0363 -0.9993>}
 head {parent=neck_02, loc=<5.7632 0.5552 0>, rot=<0 0 0.115 -0.9934>}
 dfrm_clavicle_l {parent=spine_05, loc=<2.5856 -0.2866 -4.3777>, rot=<0.0465 -0.7289 -0.009 -0.6829>}
 dfrm_upperarm_l {parent=dfrm_clavicle_l, loc=<18.2264 0.1342 -0.001>, rot=<0.0253 -0.2693 0.0013 -0.9627>}
 dfrm_lowerarm_l {parent=dfrm_upperarm_l, loc=<29.0071 0.1686 -1.239>, rot=<0.0391 -0.0099 -0.2453 -0.9686>}
 dfrm_lowerarm_twist_01_l {parent=dfrm_lowerarm_l, loc=<17.8549 0 0>, rot=<0 0 0 1>}
 dfrm_lowerarm_twist_01_a_l {parent=dfrm_lowerarm_twist_01_l, loc=<0 0 0>, rot=<0 0 0 1>}
 dfrm_lowerarm_twist_02_l {parent=dfrm_lowerarm_l, loc=<9.6199 0 0>, rot=<0 0 0 1>}
 dfrm_hand_l {parent=dfrm_lowerarm_l, loc=<25.8607 0.222 1.0274>, rot=<-0.7193 -0.0988 0.1193 -0.6773>}
 drm_ArmFeather_02_l_jnt {parent=dfrm_lowerarm_l, loc=<2.0507 -13.2978 -0.2203>, rot=<-0.0542 -0.0516 0.3423 -0.9366>}
 drm_ArmFeather_03_l_jnt {parent=dfrm_lowerarm_l, loc=<14.375 -9.4807 -0.8462>, rot=<-0.0558 -0.0194 0.1991 -0.9782>}
 dfrm_upperarm_twist_01_l {parent=dfrm_upperarm_l, loc=<7.9796 0.0843 -0.6195>, rot=<0 0 0 1>}
 dfrm_upperarm_twist_02_l {parent=dfrm_upperarm_l, loc=<19.5059 0.0843 -0.6195>, rot=<0 0 0 1>}
 drm_ArmFeather_01_l_jnt {parent=dfrm_upperarm_l, loc=<21.7807 -14.5023 -2.5933>, rot=<-0.062 0.0548 0.2181 -0.9724>}
 dfrm_clavicle_r {parent=spine_05, loc=<2.5856 -0.2865 4.3777>, rot=<0.7289 0.0465 -0.6829 0.009>}
 dfrm_upperarm_r {parent=dfrm_clavicle_r, loc=<-18.2264 -0.1342 0.001>, rot=<0.0253 -0.2693 0.0013 -0.9627>}
 dfrm_lowerarm_r {parent=dfrm_upperarm_r, loc=<-29.0072 -0.1686 1.239>, rot=<0.0391 -0.0099 -0.2453 -0.9686>}
 dfrm_lowerarm_twist_01_r {parent=dfrm_lowerarm_r, loc=<-17.855 0 0>, rot=<0 0 0 1>}
 dfrm_lowerarm_twist_02_r {parent=dfrm_lowerarm_r, loc=<-9.6199 0 0>, rot=<0 0 0 1>}
 dfrm_hand_r {parent=dfrm_lowerarm_r, loc=<-25.8607 -0.222 -1.0274>, rot=<-0.7193 -0.0988 0.1193 -0.6773>}
 drm_ArmFeather_02_r_jnt {parent=dfrm_lowerarm_r, loc=<-2.0509 13.2979 0.2207>, rot=<-0.3423 -0.9366 -0.0542 0.0516>}
 drm_ArmFeather_03_r_jnt {parent=dfrm_lowerarm_r, loc=<-14.3749 9.4806 0.8461>, rot=<-0.1991 -0.9782 -0.0558 0.0194>}
 dfrm_upperarm_twist_01_r {parent=dfrm_upperarm_r, loc=<-7.9796 -0.0843 0.6195>, rot=<0 0 0 1>}
 dfrm_upperarm_twist_02_r {parent=dfrm_upperarm_r, loc=<-19.506 -0.0843 0.6195>, rot=<0 0 0 1>}
 dfrm_upperarm_clamp_r_ {parent=dfrm_upperarm_r, loc=<-29.0072 -0.1686 1.239>, rot=<0.0391 -0.0099 -0.2453 -0.9686>}
 drm_ArmFeather_01_r_jnt {parent=dfrm_upperarm_r, loc=<-21.7806 14.5023 2.593>, rot=<-0.2181 -0.9724 -0.062 -0.0548>}
 attach_cape {parent=spine_05, loc=<9.1509 -13.3783 0>, rot=<0.0616 -0.7044 -0.0616 -0.7044>}
 attach_backpack {parent=spine_05, loc=<-0.123 -15.1423 0>, rot=<-0.0093 -0.707 0.0093 -0.707>}
 dfrm_spine_01_l {parent=spine_01, loc=<0 0 0>, rot=<0 0 0 1>}
 dfrm_spine_01_r {parent=spine_01, loc=<0 0 0>, rot=<0 0 0 1>}
 dfrm_spine_01_front {parent=spine_01, loc=<0 0 0>, rot=<0 0 0 1>}
 dfrm_spine_01_back {parent=spine_01, loc=<0 0 0>, rot=<0 0 0 1>}
 thigh_l {parent=pelvis, loc=<-1.8716 -0.0109 -11.9936>, rot=<0.0012 0.0572 0.0155 -0.9982>}
 calf_l {parent=thigh_l, loc=<-41.9178 -0.2246 0.0926>, rot=<0.0006 -0.0161 -0.0391 -0.9991>}
 foot_l {parent=calf_l, loc=<-38.7057 1.2438 0.2452>, rot=<0.0975 -0.0474 0.0317 -0.9936>}
 ball_l {parent=foot_l, loc=<-6.1614 13.1469 0.6807>, rot=<0.0001 -0.0001 -0.7063 -0.7079>}
 calf_twist_01_l {parent=calf_l, loc=<-32.9335 -0.0967 0.0397>, rot=<0.0237 0.0095 -0.0035 -0.9997>}
 calf_twist_02_l {parent=calf_l, loc=<-15.9268 -0.0967 0.0397>, rot=<0.0237 0.0095 -0.0035 -0.9997>}
 thigh_twist_01_l {parent=thigh_l, loc=<-21.6303 0 0>, rot=<0.0013 0 0 -1>}
 dyn_thigh_shorts_l {parent=thigh_twist_01_l, loc=<0 0 0>, rot=<0 0 0 1>}
 thigh_r {parent=pelvis, loc=<-1.8716 -0.0109 11.9936>, rot=<0.0572 -0.0012 0.9982 0.0155>}
 calf_r {parent=thigh_r, loc=<41.9178 0.2246 -0.0926>, rot=<0.0006 -0.0161 -0.0391 -0.9991>}
 foot_r {parent=calf_r, loc=<38.7057 -1.2438 -0.2452>, rot=<0.0975 -0.0474 0.0317 -0.9936>}
 ball_r {parent=foot_r, loc=<6.1614 -13.1469 -0.6807>, rot=<0.0001 -0.0001 -0.7063 -0.7079>}
 calf_twist_01_r {parent=calf_r, loc=<32.9335 0.0967 -0.0397>, rot=<0.0237 0.0095 -0.0035 -0.9997>}
 calf_twist_02_r {parent=calf_r, loc=<15.9268 0.0967 -0.0397>, rot=<0.0237 0.0095 -0.0035 -0.9997>}
 thigh_twist_01_r {parent=thigh_r, loc=<21.6303 0 0>, rot=<0.0013 0 0 -1>}
 dyn_thigh_shorts_r {parent=thigh_twist_01_r, loc=<-0.0002 0 0>, rot=<0 0 0 1>}
 dyn_fat_bottom_r_Base {parent=pelvis, loc=<-36.9007 -6.2619 25.5934>, rot=<-0.0305 -0.2985 -0.063 -0.9518>}
 dyn_fat_bottom_r {parent=dyn_fat_bottom_r_Base, loc=<0 0 0>, rot=<0 0 0 1>}
 dyn_fat_bottom_l_Base {parent=pelvis, loc=<-36.9006 -6.2619 -25.5934>, rot=<-0.2985 0.0305 0.9518 -0.063>}
 dyn_fat_bottom_l {parent=dyn_fat_bottom_l_Base, loc=<0 0 0>, rot=<0 0 0 1>}
 dfrm_pelvis {parent=pelvis, loc=<0 0 0>, rot=<0 0 0 1>}
 dfrm_stomach_mid {parent=dfrm_pelvis, loc=<-0.6496 26.7703 0>, rot=<0 0 -0.6943 -0.7197>}
 dfrm_stomach_01_l {parent=dfrm_pelvis, loc=<-0.889 23.2338 -16.3505>, rot=<0.2086 -0.2162 -0.6623 -0.6864>}
 dfrm_stomach_02_l {parent=dfrm_pelvis, loc=<-1.4958 12.4449 -29.0464>, rot=<0.3799 -0.3964 -0.5796 -0.6021>}
 dfrm_stomach_03_l {parent=dfrm_pelvis, loc=<-0.8319 -4.4668 -34.1516>, rot=<0.5236 -0.5427 -0.456 -0.4727>}
 dfrm_stomach_01_r {parent=dfrm_pelvis, loc=<-0.889 23.2338 16.3505>, rot=<-0.2086 0.2162 -0.6623 -0.6864>}
 dfrm_stomach_02_r {parent=dfrm_pelvis, loc=<-1.4958 12.4449 29.0464>, rot=<-0.3799 0.3964 -0.5796 -0.6021>}
 dfrm_stomach_03_r {parent=dfrm_pelvis, loc=<-0.832 -4.4668 34.1516>, rot=<0.5236 -0.5427 0.456 0.4727>}
 dfrm_stomach_06_r {parent=dfrm_pelvis, loc=<-18.5253 -3.4985 36.2463>, rot=<0.3919 -0.6264 0.5772 0.3476>}
 dfrm_stomach_05_r {parent=dfrm_pelvis, loc=<-19.1119 14.2418 30.2543>, rot=<0.2744 -0.4562 0.735 0.4199>}
 dfrm_stomach_04_l {parent=dfrm_pelvis, loc=<-19.4899 25.0742 -16.8238>, rot=<-0.1344 0.2686 0.8371 0.4571>}
 dfrm_stomach_05_l {parent=dfrm_pelvis, loc=<-19.1119 14.2418 -30.2543>, rot=<-0.2744 0.4562 0.735 0.4199>}
 dfrm_stomach_06_l {parent=dfrm_pelvis, loc=<-18.5253 -3.4985 -36.2463>, rot=<0.3919 -0.6264 -0.5772 -0.3476>}
 dfrm_stomach_lower_mid {parent=dfrm_pelvis, loc=<-19.5772 28.5457 0>, rot=<0 0 0.8843 0.4669>}
 dfrm_stomach_04_r {parent=dfrm_pelvis, loc=<-19.4899 25.0742 16.8238>, rot=<0.1344 -0.2686 0.8371 0.4571>}
 dfrm_pelvis_ctrl {controlType=EULER_TRANSFORM, parent=dfrm_pelvis, type=Control, loc=<0 0 0>, rot=<0 0 0 1>}
 dfrm_pelvis_ctrl_3 {controlType=EULER_TRANSFORM, parent=dfrm_pelvis, type=Control, loc=<0 0 0>, rot=<0 0 0 1>}
 dyn_tailFeatherMain_MSTR {parent=dfrm_pelvis, loc=<-3.1264 -33.4286 0>, rot=<0.8451 -0.5346 0 0>}
 dyn_tailFeatherMain_01 {parent=dyn_tailFeatherMain_MSTR, loc=<0 0 0>, rot=<0 0 0 1>}
 dyn_tailFeatherMain_02 {parent=dyn_tailFeatherMain_01, loc=<27.7583 0 0>, rot=<0 0 -0.0249 -0.9997>}
 dyn_tailFeather_C_01 {parent=dyn_tailFeatherMain_02, loc=<10.9998 17.0134 0>, rot=<-0.4632 -0.8863 0 0>}
 dyn_tailFeather_C_02 {parent=dyn_tailFeather_C_01, loc=<14.7905 0 0>, rot=<0 0 0.0661 -0.9978>}
 dyn_tailFeather_D_01 {parent=dyn_tailFeatherMain_02, loc=<24.3745 14.7379 0>, rot=<-0.5507 -0.8347 0 0>}
 dyn_tailFeather_D_02 {parent=dyn_tailFeather_D_01, loc=<9.7454 0 0>, rot=<0 0 0.1434 -0.9897>}
 dyn_tailFeather_A_01 {parent=dyn_tailFeatherMain_01, loc=<7.7342 16.8018 0.0001>, rot=<-0.5105 -0.8599 -0.0002 -0.0001>}
 dyn_tailFeather_A_02 {parent=dyn_tailFeather_A_01, loc=<10.9106 0 0>, rot=<-0.0003 0 0.0975 -0.9952>}
 dyn_tailFeather_B_01 {parent=dyn_tailFeatherMain_01, loc=<22.0542 17.5904 0>, rot=<-0.4987 -0.8668 0 0>}
 dyn_tailFeather_B_02 {parent=dyn_tailFeather_B_01, loc=<12.6962 0 0>, rot=<0 0 0.1681 -0.9858>}
 pelvis_ctrl {controlType=EULER_TRANSFORM, type=Control, loc=<0 0 0>, rot=<0 0 0 1>}
]
variables [
 Is Crouching Alpha {type=double, default=0}
 Is Melee OneHanded Alpha {type=double, default=0}
 Is Pistol Targeting Alpha {type=double, default=0}
 Is Playing Melee DualWield Anim Alpha {type=double, default=0}
 Is Rifle Targeting Alpha {type=double, default=0}
 SPR_upperarm_l_down_A_alpha {type=double, default=0}
 SPR_upperarm_r_down_A_alpha {type=double, default=0}
 SPR_upperarm_r_down_B_alpha {type=double, default=0}
 mult_spr_dst_hand_r_thigh_twist_r_alpha {type=double, default=0}
 mult_spr_dst_lowerarm_r_pelvis_alpha {type=double, default=0}
]
