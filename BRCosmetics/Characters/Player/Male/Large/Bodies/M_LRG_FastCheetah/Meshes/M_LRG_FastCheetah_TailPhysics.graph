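# Physics asset for the tail of the large male FastCheetah player body cosmetic, giving its tail feather bones capsule bodies for simulated secondary motion.
# A chain runs from pelvis through dfrm_pelvis and a master bone into two main segments, with two-bone feathers A and B branching from the first and C and D from the second.
# The two pelvis-to-master links swing and twist freely, while tail joints carry tight angular limits, slightly wider on the side feathers.
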PhysicsAsset /BRCosmetics/Characters/Player/Male/Large/Bodies/M_LRG_FastCheetah/Meshes/M_LRG_FastCheetah_TailPhysics
bodies {
 dfrm_pelvis [{shapeType=Capsule, length=5, radius=12, shapeName=dfrm_pelvis_capsule}]
 dyn_tailFeatherMain_01 [{shapeType=Capsule, length=19.8137, radius=14.7599, center=<15.3473 0 -0.0007>, rotation=<rot 90 0 0>}]
 dyn_tailFeatherMain_02 [{shapeType=Capsule, length=9.4615, radius=17.3357, center=<14.5908 0 0.0005>, rotation=<rot 90 0 0>}]
 dyn_tailFeatherMain_MSTR [{shapeType=Capsule, length=5, radius=12, shapeName=dyn_tailFeatherMain_MSTR_capsule}]
 dyn_tailFeather_A_01 [{shapeType=Capsule, length=7, radius=4.5, center=<5 1 0>, rotation=<rot 90 0 10>}]
 dyn_tailFeather_A_02 [{shapeType=Capsule, length=6.6415, radius=4.6166, center=<5 0 0>, rotation=<rot 90 0 0>}]
 dyn_tailFeather_B_01 [{shapeType=Capsule, length=8, radius=5.75, center=<5 1.5 0>, rotation=<rot 90 0 15>}]
 dyn_tailFeather_B_02 [{shapeType=Capsule, length=6.3875, radius=5.2135, center=<5 0 0>, rotation=<rot 90 0 0>}]
 dyn_tailFeather_C_01 [{shapeType=Capsule, length=8, radius=5.5, center=<5 1 0>, rotation=<rot 90 0 0>}]
 dyn_tailFeather_C_02 [{shapeType=Capsule, length=8, radius=4, center=<5 1 0>, rotation=<rot 90 0 0>}]
 dyn_tailFeather_D_01 [{shapeType=Capsule, length=4.5213, radius=4.3624, center=<5 0 0>, rotation=<rot 90 0 0>}]
 dyn_tailFeather_D_02 [{shapeType=Capsule, length=7.0212, radius=3.7768, center=<6.5081 0 0>, rotation=<rot 90 0 0>}]
 pelvis [{shapeType=Capsule, length=4.9046, radius=35.2946, center=<-10.9162 -4.9547 -0.1765>, rotation=<rot 0.3009 -90 2.6868>}]
}
constraints [
 {bone1Name=dyn_tailFeatherMain_01, bone2Name=dyn_tailFeatherMain_MSTR, swing1LimitDegrees=8, swing2LimitDegrees=20, twistLimitDegrees=4}
 {bone1Name=dyn_tailFeatherMain_02, bone2Name=dyn_tailFeatherMain_01, swing1LimitDegrees=8, swing2LimitDegrees=20, twistLimitDegrees=4}
 {bone1Name=dyn_tailFeather_C_01, bone2Name=dyn_tailFeatherMain_02, swing1LimitDegrees=15, swing2LimitDegrees=20, twistLimitDegrees=4}
 {bone1Name=dyn_tailFeather_C_02, bone2Name=dyn_tailFeather_C_01, swing1LimitDegrees=15, swing2LimitDegrees=20, twistLimitDegrees=4}
 {bone1Name=dyn_tailFeather_D_01, bone2Name=dyn_tailFeatherMain_02, swing1LimitDegrees=15, swing2LimitDegrees=20, twistLimitDegrees=4}
 {bone1Name=dyn_tailFeather_D_02, bone2Name=dyn_tailFeather_D_01, swing1LimitDegrees=15, swing2LimitDegrees=20, twistLimitDegrees=4}
 {bone1Name=dyn_tailFeather_A_01, bone2Name=dyn_tailFeatherMain_01, swing1LimitDegrees=15, swing2LimitDegrees=20, twistLimitDegrees=4}
 {bone1Name=dyn_tailFeather_A_02, bone2Name=dyn_tailFeather_A_01, swing1LimitDegrees=15, swing2LimitDegrees=20, twistLimitDegrees=4}
 {bone1Name=dyn_tailFeather_B_01, bone2Name=dyn_tailFeatherMain_01, swing1LimitDegrees=15, swing2LimitDegrees=20, twistLimitDegrees=4}
 {bone1Name=dyn_tailFeather_B_02, bone2Name=dyn_tailFeather_B_01, swing1LimitDegrees=15, swing2LimitDegrees=20, twistLimitDegrees=4}
 {
  bone1Name=dyn_tailFeatherMain_MSTR
  bone2Name=dfrm_pelvis
  swing1LimitDegrees=45
  swing1Motion=Free
  swing2LimitDegrees=45
  swing2Motion=Free
  twistLimitDegrees=45
  twistMotion=Free
 }
 {
  bone1Name=dfrm_pelvis
  bone2Name=pelvis
  swing1LimitDegrees=45
  swing1Motion=Free
  swing2LimitDegrees=45
  swing2Motion=Free
  twistLimitDegrees=45
  twistMotion=Free
 }
]
