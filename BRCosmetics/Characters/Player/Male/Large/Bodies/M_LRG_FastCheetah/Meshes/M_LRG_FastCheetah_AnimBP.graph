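# Anim Blueprint for the FastCheetah large male player body, adding secondary motion to a pose copied from the parent mesh.
# Constraints, additive bone offsets and a Control Rig adjust backpack, pelvis, shorts and feather bones, then two RigidBody nodes simulate tail and arm physics.
# Each update turns movement states, targeting and the parent anim instance's weapon type into correction alphas.
# Leg-raise alphas are computed but left unconnected.
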AnimBlueprint /BRCosmetics/Characters/Player/Male/Large/Bodies/M_LRG_FastCheetah/Meshes/M_LRG_FastCheetah_AnimBP
target_skeleton=&/M_LRG_FastCheetah_Skeleton
graphs [
 AnimGraph {nodes [
   n1 = "Constraint - attach_backpack (PPPP)"!(
    ComponentPose=Local To Component(LocalPose=Copy Pose From Mesh() {bUseAttachedParent=true})
    Alpha=1
    ConstraintWeights_0=0.1
    ConstraintWeights_1:float=0.1
    ConstraintWeights_2:float=0.45
    ConstraintWeights_3:float=0.35
   ) -> (Pose) {
    alphaBoolBlend {blendOption=Linear}
    boneToModify {boneName=attach_backpack}
    constraintSetup [
     {targetBone {boneName=dfrm_clavicle_l}}
     {targetBone {boneName=dfrm_clavicle_r}}
     {targetBone {boneName=spine_03}}
     {targetBone {boneName=spine_02}}
    ]
    constraintWeights [1,1,1,1]
   }
   n2 = "Transform (Modify) Bone - Bone: attach_backpack"(ComponentPose=@n1, Alpha=Get Is Sliding Alpha() -> (Is Sliding Alpha:double), Translation="0,-1,0", Rotation="0,0,-20", Scale="1,1,1") {
    rotationMode=BMM_Additive
    rotationSpace=BCS_BoneSpace
    translationMode=BMM_Additive
    translationSpace=BCS_BoneSpace
    alphaBoolBlend {blendOption=Linear}
    alphaScaleBiasClamp {interpSpeedIncreasing=6}
    boneToModify {boneName=attach_backpack}
    rotation {roll=-20}
    translation {y=-1}
   }
   n3 = Control Rig!(
    Source=Component To Local(ComponentPose=@n2)
    Alpha
    Is Pistol Targeting Alpha:double=Get Is Pistol Targeting Alpha() -> (Is Pistol Targeting Alpha:double)
    Is Rifle Targeting Alpha:double=Get Is Rifle Targeting Alpha() -> (Is Rifle Targeting Alpha:double)
    Is Crouching Alpha:double=Get Is Crouching Alpha() -> (Is Crouching Alpha:double)
    Is Playing Melee DualWield Anim Alpha:double=Get Is Playing Melee DualWield Anim Alpha() -> (Is Playing Melee DualWield Anim Alpha:double)
    Is Melee OneHanded Alpha:double=Get Is Playing Melee OneHanded Anim Alpha() -> (Is Playing Melee OneHanded Anim Alpha:double)
   ) -> (Pose) {
    alphaBoolBlend {blendOption=Linear}
    controlRigAssetReference {blueprintRigClass=&/M_LRG_FastCheetah_CtrlRig.M_LRG_FastCheetah_CtrlRig_C}
   }
   n4 = "Constraint - dyn_tailFeatherMain_MSTR (P)"(ComponentPose=Local To Component(LocalPose=@n3), Alpha=1, ConstraintWeights_0=1) {
    alphaBoolBlend {blendOption=Linear}
    boneToModify {boneName=dyn_tailFeatherMain_MSTR}
    constraintSetup [{targetBone {boneName=dfrm_pelvis}}]
    constraintWeights [1]
   }
   n5 = "Transform (Modify) Bone - Bone: dfrm_pelvis"(ComponentPose=@n4, Alpha=Get Is Crouching Alpha(), Translation="10,5,0", Rotation="0,0,0", Scale="1,1,1") {
    rotationMode=BMM_Additive
    rotationSpace=BCS_BoneSpace
    translationMode=BMM_Additive
    translationSpace=BCS_BoneSpace
    alphaBoolBlend {blendOption=Linear}
    alphaScaleBiasClamp {bInterpResult=true, interpSpeedIncreasing=6}
    boneToModify {boneName=dfrm_pelvis}
    translation=<10 5>
   }
   n6 = "Transform (Modify) Bone - Bone: dfrm_pelvis"(ComponentPose=@n5, Alpha=Get Is Sliding Alpha(), Translation="5,5,0", Rotation="0,30,0", Scale="1,1,1") {
    rotationMode=BMM_Additive
    rotationSpace=BCS_BoneSpace
    translationMode=BMM_Additive
    translationSpace=BCS_BoneSpace
    alphaBoolBlend {blendOption=Linear}
    alphaScaleBiasClamp {interpSpeedIncreasing=6}
    boneToModify {boneName=dfrm_pelvis}
    rotation {yaw=30}
    translation=<5 5>
   }
   n7 = "Transform (Modify) Bone - Bone: dfrm_pelvis"(
    ComponentPose=@n6
    Alpha=Get Is JumpingFalling Alpha() -> (Is JumpingFalling Alpha:double)
    Translation="10,-5,0"
    Rotation="0,-10,0"
    Scale="1,1,1"
   ) {
    rotationMode=BMM_Additive
    rotationSpace=BCS_BoneSpace
    translationMode=BMM_Additive
    translationSpace=BCS_BoneSpace
    alphaBoolBlend {blendOption=Linear}
    alphaScaleBiasClamp {interpSpeedIncreasing=6}
    boneToModify {boneName=dfrm_pelvis}
    rotation {yaw=-10}
    translation=<10 -5>
   }
   n8 = "Constraint - dyn_fat_bottom_l_Base (P)"(ComponentPose=@n7, Alpha=1, ConstraintWeights_0=1) {
    alphaBoolBlend {blendOption=Linear}
    boneToModify {boneName=dyn_fat_bottom_l_Base}
    constraintSetup [{targetBone {boneName=dfrm_pelvis}}]
    constraintWeights [1]
   }
   n9 = "Constraint - dyn_fat_bottom_r_Base (P)"(ComponentPose=@n8, Alpha=1, ConstraintWeights_0=1) {
    alphaBoolBlend {blendOption=Linear}
    boneToModify {boneName=dyn_fat_bottom_r_Base}
    constraintSetup [{targetBone {boneName=dfrm_pelvis}}]
    constraintWeights [1]
   }
   n10 = "Transform (Modify) Bone - Bone: dyn_thigh_shorts_l"(ComponentPose=@n9, Alpha=1, Translation="20,0,5", Rotation="0,0,-20", Scale="1,1,1") {
    rotationMode=BMM_Additive
    rotationSpace=BCS_BoneSpace
    translationMode=BMM_Additive
    translationSpace=BCS_BoneSpace
    alphaBoolBlend {blendOption=Linear}
    boneToModify {boneName=dyn_thigh_shorts_l}
    rotation {roll=-20}
    translation {x=20, z=5}
   }
   n11 = "Transform (Modify) Bone - Bone: dyn_thigh_shorts_r"(ComponentPose=@n10, Alpha=1, Translation="-20,0,-5", Rotation="0,0,-20", Scale="1,1,1") {
    rotationMode=BMM_Additive
    rotationSpace=BCS_BoneSpace
    translationMode=BMM_Additive
    translationSpace=BCS_BoneSpace
    alphaBoolBlend {blendOption=Linear}
    boneToModify {boneName=dyn_thigh_shorts_r}
    rotation {roll=-20}
    translation {x=-20, z=-5}
   }
   n12 = "Constraint - drm_ArmFeather_01_l_jnt (P)"(ComponentPose=@n11, Alpha=1, ConstraintWeights_0=1) {
    alphaBoolBlend {blendOption=Linear}
    boneToModify {boneName=drm_ArmFeather_01_l_jnt}
    constraintSetup [{targetBone {boneName=dfrm_upperarm_twist_02_l}}]
    constraintWeights [1]
   }
   n13 = "Constraint - drm_ArmFeather_01_r_jnt (P)"(ComponentPose=@n12, Alpha=1, ConstraintWeights_0=1) {
    alphaBoolBlend {blendOption=Linear}
    boneToModify {boneName=drm_ArmFeather_01_r_jnt}
    constraintSetup [{targetBone {boneName=dfrm_upperarm_twist_02_r}}]
    constraintWeights [1]
   }
   n14 = "Constraint - drm_ArmFeather_02_l_jnt (PP)"!(ComponentPose=@n13, Alpha=1, ConstraintWeights_0=0.5, ConstraintWeights_1:float=0.5) -> (Pose) {
    alphaBoolBlend {blendOption=Linear}
    boneToModify {boneName=drm_ArmFeather_02_l_jnt}
    constraintSetup [{targetBone {boneName=dfrm_lowerarm_l}}, {targetBone {boneName=dfrm_lowerarm_twist_02_l}}]
    constraintWeights [1,1]
   }
   n15 = "Constraint - drm_ArmFeather_02_r_jnt (PP)"!(ComponentPose=@n14, Alpha=1, ConstraintWeights_0=0.5, ConstraintWeights_1:float=0.5) -> (Pose) {
    alphaBoolBlend {blendOption=Linear}
    boneToModify {boneName=drm_ArmFeather_02_r_jnt}
    constraintSetup [{targetBone {boneName=dfrm_lowerarm_r}}, {targetBone {boneName=dfrm_lowerarm_twist_02_r}}]
    constraintWeights [1,1]
   }
   n16 = "Constraint - drm_ArmFeather_03_l_jnt (P)"(ComponentPose=@n15, Alpha=1, ConstraintWeights_0=1) {
    alphaBoolBlend {blendOption=Linear}
    boneToModify {boneName=drm_ArmFeather_03_l_jnt}
    constraintSetup [{targetBone {boneName=dfrm_lowerarm_twist_01_l}}]
    constraintWeights [1]
   }
   n17 = "Constraint - drm_ArmFeather_03_r_jnt (P)"(ComponentPose=@n16, Alpha=1, ConstraintWeights_0=1) {
    alphaBoolBlend {blendOption=Linear}
    boneToModify {boneName=drm_ArmFeather_03_r_jnt}
    constraintSetup [{targetBone {boneName=dfrm_lowerarm_twist_01_r}}]
    constraintWeights [1]
   }
   n18 = RigidBody(ComponentPose=@n17) {
    bClampLinearTranslationLimitToRefPose=true
    bOverrideWorldGravity=true
    overridePhysicsAsset=&/M_LRG_FastCheetah_TailPhysics
    simulationSpace=BaseBoneSpace
    alphaBoolBlend {blendOption=Linear}
    baseBoneRef {boneName=root}
    componentAppliedLinearAccClamp=<1500 1500 1500>
    componentLinearAccScale=<0.4 0.4 0.4>
    componentLinearVelScale=<0.2 0.2 0.2>
    overrideWorldGravity {z=-980}
   }
   n19 = RigidBody(ComponentPose=@n18) {
    bClampLinearTranslationLimitToRefPose=true
    bOverrideWorldGravity=true
    overridePhysicsAsset=&/M_LRG_FastCheetah_ArmPhysics
    simulationSpace=BaseBoneSpace
    alphaBoolBlend {blendOption=Linear}
    baseBoneRef {boneName=spine_05}
    componentAppliedLinearAccClamp=<1500 1500 1500>
    componentLinearAccScale=<0.4 0.4 0.4>
    componentLinearVelScale=<0.2 0.2 0.2>
    overrideWorldGravity {z=-980}
   }
   Output Pose(Result=Component To Local(ComponentPose=@n19)) {layerGroup=None}
   Get leftLegUp() -> (leftLegUp:double)
   Get rightLegUp() -> (rightLegUp:double)
  ]}
 EventGraph {nodes [
   Event BlueprintUpdateAnimation()
   > n1 = Sequence!(execute) -> (then_0, then_1, then_2:Exec, then_3:Exec, then_4:Exec, then_5:Exec, then_6:Exec)
   > n2 = Sequence!(execute) -> (then_0, then_1, then_2)
   n3 = Get skelMesh() -> (skelMesh:Skeletal Mesh Component Object Reference)
   > Set leftLegUp(leftLegUp:double=GetAlphaFromReferenceTransformDelta(SkeletalMeshComponent=@n3, BoneName=thigh_l, BaseName=pelvis, InputRange=(X=-22,Y=-94), TwistAxis="1,0,0")) -> (Output_Get:double)
   Reroute Node(InputPin=@n2.then_2)
   Set rightLegUp(
    execute=@n2.then_1
    rightLegUp:double=GetAlphaFromReferenceTransformDelta(SkeletalMeshComponent=@n3, BoneName=thigh_r, BaseName=pelvis, InputRange=(X=-1,Y=-94), TwistAxis="1,0,0")
   ) -> (Output_Get:double)
   Set Is Crouching Alpha(execute=@n1.then_1, Is Crouching Alpha:double="To Float (Boolean)"(InBool=Get bIsCrouching())) -> (Output_Get:double)
   Set Is Sliding Alpha(execute=@n1.then_2, Is Sliding Alpha:double="To Float (Boolean)"(InBool=Get bIsSliding())) -> (Output_Get:double)
   n4 = NOT Boolean(A=Get bIsSkyDiving())
   n5 = OR Boolean(A=Get bIsFalling(), B=Get bIsJumping())
   Set Is JumpingFalling Alpha(execute=@n1.then_3, Is JumpingFalling Alpha:double="To Float (Boolean)"(InBool=AND Boolean(A=@n5, B=@n4))) -> (Output_Get:double)
   n6 = Get bIsPlayingMeleeAnim()
   n7 = Get Is Melee Dual Wield() -> (Is Melee Dual Wield:bool)
   Set Is Playing Melee DualWield Anim Alpha(execute=@n1.then_5, Is Playing Melee DualWield Anim Alpha:double="To Float (Boolean)"(InBool=AND Boolean(A=@n6, B=@n7))) -> (Output_Get:double)
   Set Is Playing Melee OneHanded Anim Alpha(
    execute=@n1.then_6
    Is Playing Melee OneHanded Anim Alpha:double="To Float (Boolean)"(InBool=Get Is Melee One Hand() -> (Is Melee One Hand:bool))
   ) -> (Output_Get:double)
   n8 = Sequence!(execute=@n1.then_4) -> (then_0, then_1, then_2, then_3:Exec, then_4:Exec)
   > n9 = Cast To FortPlayerAnimInstance(Object=GetAnimInstance(self=GetOwnerSkeletalMeshComponent()))
   > Set isPistol(isPistol=false)
   > Set isRifle(isRifle:bool=false) -> (Output_Get:bool)
   > Set Is Melee Dual Wield(Is Melee Dual Wield:bool=false) -> (Output_Get:bool)
   > Set Is Melee One Hand(Is Melee One Hand:bool=false) -> (Output_Get:bool)
   > n10 = Switch on EFortWeaponCoreAnimation!(execute, Selection:EFortWeaponCoreAnimation Enum=Get WeaponCoreAnim(self=@n9)) -> (
    Melee:Exec
    Pistol:Exec
    Shotgun:Exec
    PaperBlueprint:Exec
    Rifle:Exec
    MeleeOneHand:Exec
    MachinePistol:Exec
    RocketLauncher:Exec
    GrenadeLauncher:Exec
    GoingCommando:Exec
    AssaultRifle:Exec
    TacticalShotgun:Exec
    SniperRifle:Exec
    TrapPlacement:Exec
    ShoulderLauncher:Exec
    AbilityDecoTool:Exec
    Crossbow:Exec
    C4:Exec
    RemoteControl:Exec
    DualWield:Exec
    AR_BullPup:Exec
    AR_ForwardGrip:Exec
    MedPackPaddles:Exec
    SMG_P90:Exec
    AR_DrumGun:Exec
    Consumable_Small:Exec
    Consumable_Large:Exec
    Balloon:Exec
    MountedTurret:Exec
    CreativeTool:Exec
    ExplosiveBow:Exec
    AshtonIndigo:Exec
    AshtonChicago:Exec
    MeleeDualWield:Exec
    Unarmed:Exec
   )
   Set isRifle(execute=[@n10.AssaultRifle, @n10.Rifle, @n10.RocketLauncher, @n10.Shotgun, @n10.SniperRifle, @n10.TacticalShotgun], isRifle=true)
   Set Is Melee Dual Wield(execute=@n10.MeleeDualWield, Is Melee Dual Wield=true)
   Set Is Melee One Hand(execute=@n10.MeleeOneHand, Is Melee One Hand=true)
   Set isPistol(execute=@n10.Pistol, isPistol=true)
   n11 = Get isRifle() -> (isRifle:bool)
   n12 = Get bIsTargeting()
   Set Is Rifle Targeting Alpha(execute=@n8.then_1, Is Rifle Targeting Alpha:double="To Float (Boolean)"(InBool=AND Boolean(A=@n11, B=@n12))) -> (Output_Get:double)
   n13 = Get isPistol() -> (isPistol:bool)
   n14 = Get bIsTargeting()
   Set Is Pistol Targeting Alpha(execute=@n8.then_2, Is Pistol Targeting Alpha:double="To Float (Boolean)"(InBool=AND Boolean(A=@n13, B=@n14))) -> (Output_Get:double)
   Event BlueprintInitializeAnimation()
   > Set skelMesh(skelMesh:Skeletal Mesh Component Object Reference=GetOwnerSkeletalMeshComponent()) -> (Output_Get:Skeletal Mesh Component Object Reference)
   AND Boolean(A=Get bIsPlayingMeleeAnim(), B=false)
  ]}
]
variables [
 Is Crouching Alpha {type=number}
 Is JumpingFalling Alpha {type=number}
 Is Melee Dual Wield {type=boolean}
 Is Melee One Hand {type=boolean}
 Is Pistol Targeting Alpha {type=number}
 Is Playing Melee DualWield Anim Alpha {type=number}
 Is Playing Melee OneHanded Anim Alpha {type=number}
 Is Rifle Targeting Alpha {type=number}
 Is Sliding Alpha {type=number}
 isPistol {type=boolean}
 isRifle {type=boolean}
 leftLegUp {type=number}
 rightLegUp {type=number}
 skelMesh {type=SkeletalMeshComponent}
]
