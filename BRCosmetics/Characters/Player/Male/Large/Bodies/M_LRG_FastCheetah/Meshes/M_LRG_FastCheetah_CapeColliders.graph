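# Physics asset holding collision shapes for the M_LRG_FastCheetah player body, named as the colliders for its cape.
# Capsules cover the pelvis, spine, clavicles and neck, tapering up the spine with two extra capsules on spine_02, and a sphere covers the head.
# Mirrored capsules sit on three arm-feather joints per side, while limb bodies carry no shapes.
# It defines no constraints, so it provides collision volumes only, without a jointed ragdoll.
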PhysicsAsset /BRCosmetics/Characters/Player/Male/Large/Bodies/M_LRG_FastCheetah/Meshes/M_LRG_FastCheetah_CapeColliders
bodies {
 calf_l []
 calf_r []
 clavicle_l [{shapeType=Capsule, length=9.0192, radius=11, center=<9.4266 -1.1806 -2.4977>, rotation=<rot 75 0 0>}]
 clavicle_r [{shapeType=Capsule, length=9.0192, radius=11, center=<-9.4266 1.1806 2.4977>, rotation=<rot 75 0 0>}]
 dfrm_pelvis [{shapeType=Capsule, length=8, radius=32, shapeName=dfrm_pelvis_capsule, center=<-12.2673 -14 0>}]
 drm_ArmFeather_01_l_jnt [{shapeType=Capsule, length=8, radius=6, shapeName=drm_ArmFeather_01_l_jnt_capsule, center=<1 1 0>, rotation=<rot 90 -59.1553 -79.1562>}]
 drm_ArmFeather_01_r_jnt [{shapeType=Capsule, length=8, radius=6, shapeName=drm_ArmFeather_01_r_jnt_capsule, center=<1 -1 0>, rotation=<rot 90 59.1553 79.1558>}]
 drm_ArmFeather_02_l_jnt [{shapeType=Capsule, length=8, radius=7, shapeName=drm_ArmFeather_02_l_jnt_capsule, center=<2 1 0>, rotation=<rot 90 0 0>}]
 drm_ArmFeather_02_r_jnt [{shapeType=Capsule, length=8, radius=7, shapeName=drm_ArmFeather_02_r_jnt_capsule, center=<2 -1 0>, rotation=<rot 90 0 0>}]
 drm_ArmFeather_03_l_jnt [{shapeType=Capsule, length=8, radius=7, shapeName=drm_ArmFeather_03_l_jnt_capsule, rotation=<rot 90 0 0>}]
 drm_ArmFeather_03_r_jnt [{shapeType=Capsule, length=8, radius=7, shapeName=drm_ArmFeather_03_r_jnt_capsule, rotation=<rot 90 0 0>}]
 foot_l []
 foot_r []
 head [{shapeType=Sphere, radius=17.3125, center=<11.028 -0.174 0>}]
 lowerarm_l []
 lowerarm_r []
 neck_02 [{shapeType=Capsule, length=13.4375, radius=7.3175, center=<-0.2642 -0.3408 0>, rotation=<rot 90 -171.1193 -175.3978>}]
 pelvis [{shapeType=Capsule, length=10, radius=32, center=<-12.2673 -10 0>}]
 spine_01 [{shapeType=Capsule, length=8, radius=30, center=<-8 -4.4493 0>}]
 spine_02 [
  {shapeType=Capsule, length=8, radius=26, center=<-1.5 -3.7993 0>}
  {shapeType=Capsule, length=29.3956, radius=11.5807, center=<2.5024 0.4992 3.9825>, rotation=<rot 90 0.0049 0.0049>}
  {shapeType=Capsule, length=29.875, radius=11.1013, center=<2.4704 0.4991 -3.9825>, rotation=<rot 90 -0.0396 -0.0396>}
 ]
 spine_03 [{shapeType=Capsule, length=8, radius=24, center=<0 -3.5993 0>}]
 spine_04 [{shapeType=Capsule, length=7, radius=21, center=<0 -3.1993 0>}]
 spine_05 [{shapeType=Capsule, length=6, radius=16, center=<-2.2795 -3.1993 0>}]
 thigh_l []
 thigh_r []
 upperarm_l []
 upperarm_r []
}
constraints []
