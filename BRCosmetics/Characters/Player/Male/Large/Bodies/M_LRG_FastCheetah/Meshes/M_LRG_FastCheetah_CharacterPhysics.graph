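# Physics asset for the FastCheetah large male player body cosmetic, giving its skeleton simplified collision bodies.
# One capsule each covers the pelvis, two spine bones and the upper arm, forearm, hand, thigh, calf and foot on both sides, while the head uses a sphere.
# Only two constraints are listed, right foot to calf and right hand to forearm, both with free swing and twist, so the bodies do not form a fully jointed chain.
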PhysicsAsset /BRCosmetics/Characters/Player/Male/Large/Bodies/M_LRG_FastCheetah/Meshes/M_LRG_FastCheetah_CharacterPhysics
bodies {
 Spine_03 [{shapeType=Capsule, length=13.2683, radius=20.5627, center=<-4.2399 1.1471 0.0069>, rotation=<rot -0.0003 88.5965 179.7999>}]
 calf_l [{shapeType=Capsule, length=38.6878, radius=7.1396, center=<-20.9094 -1.1698 -0.4407>, rotation=<rot -90 0 0>}]
 calf_r [{shapeType=Capsule, length=38.6878, radius=7.1396, center=<20.9094 1.1698 0.4407>, rotation=<rot -90 0 0>}]
 foot_l [{shapeType=Capsule, length=30.6269, radius=6.8277, shapeName=hand_l_capsule, center=<-6.3437 3.7047 0>, rotation=<rot 0 -2.6669 90>}]
 foot_r [{shapeType=Capsule, length=30.6269, radius=6.8277, shapeName=hand_l_capsule, center=<6.3437 -3.7047 0>, rotation=<rot 0 -2.6669 -90>}]
 hand_l [{shapeType=Capsule, length=14.5626, radius=6.5359, shapeName=hand_l_capsule, center=<6.9044 -1.7418 0>, rotation=<rot 90 0 0>}]
 hand_r [{shapeType=Capsule, length=14.5626, radius=6.5359, shapeName=hand_l_capsule, center=<-6.9044 1.7418 0>, rotation=<rot 90 180 0>}]
 head [{shapeType=Sphere, radius=12.5475, center=<3.9694 0.6346 0>}]
 lowerarm_l [{shapeType=Capsule, length=16.9921, radius=6.3, center=<9.7344 0 0>, rotation=<rot -90 -0.0148 0.0149>}]
 lowerarm_r [{shapeType=Capsule, length=16.9921, radius=6.3, center=<-9.7344 0 0>, rotation=<rot -90 -0.0148 0.0149>}]
 pelvis [{shapeType=Capsule, length=5.4181, radius=29.119, center=<-9.4698 -4.0514 0>, rotation=<rot 0 90 0>}]
 spine_05 [{shapeType=Capsule, length=24.3792, radius=12.6229, center=<1.4593 -1.209 0.0001>}]
 thigh_l [{shapeType=Capsule, length=31.7315, radius=8.3773, center=<-19.5847 0 0.6646>, rotation=<rot 90 -0.0035 -0.002>}]
 thigh_r [{shapeType=Capsule, length=31.6975, radius=8.4113, center=<19.5847 0.0002 -0.4822>, rotation=<rot 90 -0.0184 -0.0168>}]
 upperarm_l [{shapeType=Capsule, length=21.4806, radius=6.8068, center=<11.5687 0 -0.1937>, rotation=<rot -86.7141 -179.9905 179.9915>}]
 upperarm_r [{shapeType=Capsule, length=22.2744, radius=6.8591, center=<-11.0574 -0.0001 0.2309>, rotation=<rot -87.1989 -179.9891 179.9901>}]
}
constraints [
 {
  bone1Name=foot_r
  bone2Name=calf_r
  swing1LimitDegrees=45
  swing1Motion=Free
  swing2LimitDegrees=45
  swing2Motion=Free
  twistLimitDegrees=45
  twistMotion=Free
 }
 {
  bone1Name=hand_r
  bone2Name=lowerarm_r
  swing1LimitDegrees=45
  swing1Motion=Free
  swing2LimitDegrees=45
  swing2Motion=Free
  twistLimitDegrees=45
  twistMotion=Free
 }
]
