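# Child animation Blueprint for the FastCheetah large male player character body that manages the visibility of its tail mesh.
# When character customization loads, it caches the MiscOrTail custom part as the tail mesh and checks the equipped backbling.
# It records whether a backbling is worn and treats cape or long backblings as capes, hiding the tail when one is present and showing it otherwise.
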AnimBlueprint /BRCosmetics/Characters/Player/Male/Large/Bodies/M_LRG_FastCheetah/Meshes/Parts/M_LRG_FastCheetah_AnimBP_Child
target_skeleton=/BRCosmetics/Characters/Player/Male/Large/Bodies/M_LRG_FastCheetah/Meshes/M_LRG_FastCheetah_Skeleton
graphs [EventGraph {nodes [
   n1 = Event OnCharacterCustomizationLoaded()
   > n2 = Set Tail Mesh(Tail Mesh:Skeletal Mesh Component Budgeted Object Reference=Get Array Item with EFort Custom Part Type(Array=@n1.CustomMeshParts, PartType=MiscOrTail)) -> (Output_Get:Skeletal Mesh Component Budgeted Object Reference)
   n3 = CheckForBackbling()
   > Set Is Wearing Backbling(Is Wearing Backbling=@n3)
   > n4 = Set Is Wearing A Cape(Is Wearing A Cape:bool=OR Boolean(A=@n3.bIsBackblingACape, B=@n3.bIsBackblingLong)) -> (Output_Get:bool)
   > n5 = Branch(Condition=@n4)
   > SetVisibility!(execute, self:Scene Component Object Reference=@n2, bNewVisibility:bool=false, bPropagateToChildren:bool=false) -> (then)
   SetVisibility(execute=@n5.else, self=@n2, bNewVisibility=true, bPropagateToChildren=false)
  ]}]
variables [
 Is Wearing A Cape {type=boolean}
 Is Wearing Backbling {type=boolean}
 Tail Mesh {type=SkeletalMeshComponentBudgeted}
]
