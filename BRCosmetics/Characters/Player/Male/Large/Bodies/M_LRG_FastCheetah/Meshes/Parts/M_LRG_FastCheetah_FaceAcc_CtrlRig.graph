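# Control Rig for the face accessory part of the large male FastCheetah player body, rotating the two wattles in response to jaw pose.
# In the forward solve, a spherical pose reader measures C_jaw orientation relative to the head.
# Its output weights two Modify Transforms nodes that apply the same 45-degree Z rotation to the left and right wattle master bones.
# Comb chains are in the hierarchy but are not driven.
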ControlRigBlueprint /BRCosmetics/Characters/Player/Male/Large/Bodies/M_LRG_FastCheetah/Meshes/Parts/M_LRG_FastCheetah_FaceAcc_CtrlRig
graphs [
 RigVMFunctionLibrary {links [], nodes []}
 RigVMModel {nodes [
   RigUnit_BeginExecution = &RigUnit_BeginExecution()
   n1 = &RigUnit_SphericalPoseReader(
    ActiveRegionScaleFactors=(PositiveWidth=2,NegativeWidth=2,PositiveHeight=0.5,NegativeHeight=0.5)
    ActiveRegionSize=0.1
    Debug=(bDrawDebug=false,DebugScale=25,DebugSegments=20,DebugThickness=0.25)
    DriverAxis=(X=0,Y=1,Z=0)
    DriverItem=(Type=Bone,Name="C_jaw")
    FalloffRegionScaleFactors=(PositiveWidth=1,NegativeWidth=1,PositiveHeight=0.5,NegativeHeight=0.5)
    FalloffSize=0.25
    FlipWidthScaling=false
    OptionalParentItem=(Type=Bone,Name="head")
    RotationOffset=(X=115,Y=0,Z=0)
    ExecutePin=@RigUnit_BeginExecution.ExecutePin
   )
   > &RigUnit_ModifyTransforms(
    ItemToModify=((Item=(Type=Bone,Name="Wattle_l_MSTR"),Transform=(Rotation=(X=0,Y=0,Z=0,W=1),Translation=(X=0,Y=0,Z=0),Scale3D=(X=1,Y=1,Z=1))))
    ItemToModify.0.Transform.Rotation=&RigVMFunction_MathQuaternionFromEuler(Euler=(X=0,Y=0,Z=45)).Result
    Weight=@n1.OutputParam
   )
   > &RigUnit_ModifyTransforms(
    ItemToModify=((Item=(Type=Bone,Name="Wattle_r_MSTR"),Transform=(Rotation=(X=0,Y=0,Z=0,W=1),Translation=(X=0,Y=0,Z=0),Scale3D=(X=1,Y=1,Z=1))))
    ItemToModify.0.Transform.Rotation=&RigVMFunction_MathQuaternionFromEuler(Euler=(X=0,Y=0,Z=45)).Result
    Weight=@n1.OutputParam
   )
  ]}
]
hierarchy [
 root {loc=<0 0 0>, rot=<0 0 0 1>}
 pelvis {parent=root, loc=<0 -0.7321 92.4218>, rot=<0.0127 0.707 -0.0127 -0.707>}
 spine_01 {parent=pelvis, loc=<5.6002 -0.5128 0>, rot=<0 0 0.034 -0.9994>}
 spine_02 {parent=spine_01, loc=<8.0753 0.0139 0>, rot=<0 0 0.0238 -0.9997>}
 spine_03 {parent=spine_02, loc=<7.0393 0 0>, rot=<0 0 0.0149 -0.9999>}
 spine_04 {parent=spine_03, loc=<8.1024 0 0>, rot=<0 0 -0.0132 -0.9999>}
 spine_05 {parent=spine_04, loc=<10.0146 0 0>, rot=<0 0 -0.0015 -1>}
 neck_01 {parent=spine_05, loc=<10.815 1.9196 0>, rot=<0 0 -0.1864 -0.9825>}
 neck_02 {parent=neck_01, loc=<4.7347 0.3045 0>, rot=<0 0 0.0363 -0.9993>}
 head {parent=neck_02, loc=<5.7632 0.5552 0>, rot=<0 0 0.115 -0.9934>}
 Comb_bk_MSTR {parent=head, loc=<16.3374 -8.3191 -0.0291>, rot=<-0.0097 -0.0262 0.3796 -0.9247>}
 Comb_bk_01_jnt {parent=Comb_bk_MSTR, loc=<0 0 0>, rot=<0 0 0 1>}
 Comb_bk_02_jnt {parent=Comb_bk_01_jnt, loc=<2.84 0 0>, rot=<0 0 -0.1136 -0.9935>}
 Comb_bk_03_jnt {parent=Comb_bk_02_jnt, loc=<2.9529 0 0>, rot=<0 0 0 1>}
 Comb_bk_tp_MSTR {parent=head, loc=<22.0745 -1.3681 0.0405>, rot=<-0.1254 -0.1727 0.3225 -0.9222>}
 Comb_bk_tp_01_jnt {parent=Comb_bk_tp_MSTR, loc=<0 0 0>, rot=<0 0 0 1>}
 Comb_bk_tp_02_jnt {parent=Comb_bk_tp_01_jnt, loc=<1.8308 0 0>, rot=<0.0107 -0.0257 -0.3845 -0.9227>}
 Comb_bk_tp_03_jnt {parent=Comb_bk_tp_02_jnt, loc=<3.5186 0 0>, rot=<0.0023 -0.0278 -0.0842 -0.9961>}
 Comb_bk_tp_04_jnt {parent=Comb_bk_tp_03_jnt, loc=<3.9717 0 0>, rot=<0 0 0 1>}
 Comb_ft_MSTR {parent=head, loc=<20.1724 5.4718 1.825>, rot=<0.4304 0.561 0.561 0.4304>}
 Comb_ft_01_jnt {parent=Comb_ft_MSTR, loc=<0 0 0>, rot=<0 0 0 1>}
 Comb_ft_02_jnt {parent=Comb_ft_01_jnt, loc=<3.3488 0 0>, rot=<0 0 0 1>}
 Comb_ft_tp_MSTR {parent=head, loc=<24.5056 3.779 1.9421>, rot=<-0.7793 -0.1393 -0.304 -0.53>}
 Comb_ft_tp_01_jnt {parent=Comb_ft_tp_MSTR, loc=<0 0 0>, rot=<0 0 0 1>}
 Comb_ft_tp_02_jnt {parent=Comb_ft_tp_01_jnt, loc=<2.4789 0 0>, rot=<0.0045 -0.2518 0.0212 -0.9675>}
 Comb_ft_tp_03_jnt {parent=Comb_ft_tp_02_jnt, loc=<3.3682 0 0>, rot=<0.0017 -0.1863 0.0121 -0.9824>}
 Comb_ft_tp_04_jnt {parent=Comb_ft_tp_03_jnt, loc=<3.2884 0 0>, rot=<-0.0001 -0.0924 -0.0007 -0.9957>}
 Comb_ft_tp_05_jnt {parent=Comb_ft_tp_04_jnt, loc=<3.2884 0 0>, rot=<-0.001 -0.0742 -0.0133 -0.9972>}
 Comb_ft_tp_06_jnt {parent=Comb_ft_tp_05_jnt, loc=<3.2884 0 0>, rot=<0 0 0 1>}
 faceAttach {parent=head, loc=<-4.146 -0.8545 0>, rot=<0.0029 -0.7071 -0.0029 -0.7071>}
 C_jaw {parent=faceAttach, loc=<-0.0134 6.8141 1.7316>, rot=<0 0 0 1>}
 C_lip_lower_mid {parent=C_jaw, loc=<0.0223 16.1451 -1.04>, rot=<0 0 0 -1>}
 L_lip_lower_outer {parent=C_jaw, loc=<3.7606 11.9793 -1.0575>, rot=<-0.0416 -0.0513 0.4403 -0.8954>}
 R_lip_lower_outer {parent=C_jaw, loc=<-3.7337 11.9793 -1.0575>, rot=<-0.0416 0.0513 -0.4403 -0.8954>}
 Wattle_r_MSTR {parent=C_jaw, loc=<-2.2377 7.4244 -2.1303>, rot=<0 -0.7071 0 -0.7071>}
 Wattle_r_01_jnt {parent=Wattle_r_MSTR, loc=<0 0 0>, rot=<-0.0581 -0.138 -0.0275 -0.9883>}
 Wattle_r_02_jnt {parent=Wattle_r_01_jnt, loc=<3.6362 0 0>, rot=<-0.0112 -0.0266 -0.113 -0.9932>}
 Wattle_r_03_jnt {parent=Wattle_r_02_jnt, loc=<4.3211 0 0>, rot=<0 0 -0.0565 -0.9984>}
 Wattle_l_MSTR {parent=C_jaw, loc=<1.5213 7.7769 -2.021>, rot=<0 -0.7071 0 -0.7071>}
 Wattle_l_01_jnt {parent=Wattle_l_MSTR, loc=<0 0 0>, rot=<-0.0955 0.017 -0.0069 -0.9953>}
 Wattle_l_02_jnt {parent=Wattle_l_01_jnt, loc=<5.0468 0 0>, rot=<-0.0112 -0.0266 -0.113 -0.9932>}
 Wattle_l_03_jnt {parent=Wattle_l_02_jnt, loc=<6.3996 0 0>, rot=<0 0 -0.0565 -0.9984>}
 L_lip_corner {parent=faceAttach, loc=<7.1515 14.119 -0.5943>, rot=<0.0043 0.0073 -0.1916 -0.9814>}
 R_lip_corner {parent=faceAttach, loc=<-7.1515 14.119 -0.5943>, rot=<0.0043 -0.0073 0.1916 -0.9814>}
]
