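# Physics asset for the face accessory of the large male FastCheetah character body, covering a segmented comb and a pair of wattles.
# Capsule bodies form four comb chains rooted on the head and two wattle chains rooted on the jaw, with head and jaw linking back to spine_05.
# Root links rotate freely, while chain joints have tight twist and narrow swing limits, looser on the wattles than the comb.
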PhysicsAsset /BRCosmetics/Characters/Player/Male/Large/Bodies/M_LRG_FastCheetah/Meshes/Parts/M_LRG_FastCheetah_FaceAcc_Physics
bodies {
 C_jaw [{shapeType=Capsule, length=1, radius=1, shapeName=C_jaw_capsule, rotation=<rot 90 0 0>}]
 Comb_bk_01_jnt [{shapeType=Capsule, length=2, radius=3.5, shapeName=Comb_bk_01_jnt_capsule, center=<3 0 0>, rotation=<rot 90 0 0>}]
 Comb_bk_02_jnt [{shapeType=Capsule, length=2, radius=3.5, shapeName=Comb_bk_02_jnt_capsule, center=<3 0 0>, rotation=<rot 90 0 0>}]
 Comb_bk_03_jnt [{shapeType=Capsule, length=2, radius=3.5, shapeName=Comb_bk_03_jnt_capsule, center=<3 0 0>, rotation=<rot 90 0 0>}]
 Comb_bk_MSTR [{shapeType=Capsule, length=1, radius=1, shapeName=Comb_bk_MSTR_capsule, rotation=<rot 90 0 0>}]
 Comb_bk_tp_01_jnt [{shapeType=Capsule, length=2, radius=5, shapeName=Comb_bk_tp_01_jnt_capsule, center=<3 0 0>, rotation=<rot 90 0 0>}]
 Comb_bk_tp_02_jnt [{shapeType=Capsule, length=2, radius=5, shapeName=Comb_bk_tp_02_jnt_capsule, center=<3 0 0>, rotation=<rot 90 0 0>}]
 Comb_bk_tp_03_jnt [{shapeType=Capsule, length=2, radius=5, shapeName=Comb_bk_tp_03_jnt_capsule, center=<3 0 0>, rotation=<rot 90 0 0>}]
 Comb_bk_tp_04_jnt [{shapeType=Capsule, length=2, radius=2, shapeName=Comb_bk_tp_04_jnt_capsule, center=<3 0 0>, rotation=<rot 90 0 0>}]
 Comb_bk_tp_MSTR [{shapeType=Capsule, length=1, radius=1, shapeName=Comb_bk_tp_MSTR_capsule, rotation=<rot 90 0 0>}]
 Comb_ft_01_jnt [{shapeType=Capsule, length=2, radius=2.5, shapeName=Comb_ft_01_jnt_capsule, center=<3 0 0>, rotation=<rot 90 0 0>}]
 Comb_ft_02_jnt [{shapeType=Capsule, length=2, radius=2.5, shapeName=Comb_ft_02_jnt_capsule, center=<3 0 0>, rotation=<rot 90 0 0>}]
 Comb_ft_MSTR [{shapeType=Capsule, length=1, radius=1, shapeName=Comb_ft_MSTR_capsule, rotation=<rot 90 0 0>}]
 Comb_ft_tp_01_jnt [{shapeType=Capsule, length=2, radius=3.5, shapeName=Comb_ft_tp_01_jnt_capsule, center=<3 0 0>, rotation=<rot 90 0 0>}]
 Comb_ft_tp_02_jnt [{shapeType=Capsule, length=2, radius=5, shapeName=Comb_ft_tp_02_jnt_capsule, center=<3 0 0>, rotation=<rot 90 0 0>}]
 Comb_ft_tp_03_jnt [{shapeType=Capsule, length=2, radius=6.5, shapeName=Comb_ft_tp_03_jnt_capsule, center=<3 0 0>, rotation=<rot 90 0 0>}]
 Comb_ft_tp_04_jnt [{shapeType=Capsule, length=2, radius=7, shapeName=Comb_ft_tp_04_jnt_capsule, center=<3 0 0>, rotation=<rot 90 0 0>}]
 Comb_ft_tp_05_jnt [{shapeType=Capsule, length=2, radius=7, shapeName=Comb_ft_tp_05_jnt_capsule, center=<3 0 0>, rotation=<rot 90 0 0>}]
 Comb_ft_tp_06_jnt [{shapeType=Capsule, length=2, radius=7, shapeName=Comb_ft_tp_06_jnt_capsule, center=<3 0 0>, rotation=<rot 90 0 0>}]
 Comb_ft_tp_MSTR [{shapeType=Capsule, length=1, radius=1, shapeName=Comb_ft_tp_MSTR_capsule, rotation=<rot 90 0 0>}]
 Wattle_l_01_jnt [{shapeType=Capsule, length=2, radius=2, shapeName=Wattle_l_01_jnt_capsule, center=<4 0 0>, rotation=<rot 90 0 0>}]
 Wattle_l_02_jnt [{shapeType=Capsule, length=2, radius=3, shapeName=Wattle_l_02_jnt_capsule, center=<5 0 0>, rotation=<rot 90 0 0>}]
 Wattle_l_03_jnt [{shapeType=Capsule, length=2, radius=3, shapeName=Wattle_l_03_jnt_capsule, center=<6 1 0>, rotation=<rot 90 0 0>}]
 Wattle_l_MSTR [{shapeType=Capsule, length=1, radius=1, shapeName=Wattle_l_MSTR_capsule, rotation=<rot 90 0 0>}]
 Wattle_r_01_jnt [{shapeType=Capsule, length=2, radius=2, shapeName=Wattle_r_01_jnt_capsule, center=<4 0 0>, rotation=<rot 90 0 0>}]
 Wattle_r_02_jnt [{shapeType=Capsule, length=2, radius=3, shapeName=Wattle_r_02_jnt_capsule, center=<5 0 0>, rotation=<rot 90 0 0>}]
 Wattle_r_03_jnt [{shapeType=Capsule, length=2, radius=3, shapeName=Wattle_r_03_jnt_capsule, center=<6 0 0>, rotation=<rot 90 0 0>}]
 Wattle_r_MSTR [{shapeType=Capsule, length=1, radius=1, shapeName=Wattle_r_MSTR_capsule, rotation=<rot 90 0 0>}]
 head [{shapeType=Capsule, length=5, radius=13, shapeName=head_capsule, center=<4.0479 1.9384 0>, rotation=<rot 90 0 0>}]
 spine_05 [
  {shapeType=Capsule, length=5, radius=12, shapeName=spine_05_capsule, center=<1.0072 5.148 0>}
  {shapeType=Capsule, length=5, radius=12, shapeName=spine_05_capsule, center=<-9.7326 10.1035 0>}
 ]
}
constraints [
 {
  bone1Name=Comb_bk_MSTR
  bone2Name=head
  swing1LimitDegrees=45
  swing1Motion=Free
  swing2LimitDegrees=45
  swing2Motion=Free
  twistLimitDegrees=45
  twistMotion=Free
 }
 {bone1Name=Comb_bk_01_jnt, bone2Name=Comb_bk_MSTR, swing1LimitDegrees=15, swing2LimitDegrees=15, twistLimitDegrees=2}
 {bone1Name=Comb_bk_02_jnt, bone2Name=Comb_bk_01_jnt, swing1LimitDegrees=15, swing2LimitDegrees=15, twistLimitDegrees=2}
 {bone1Name=Comb_bk_03_jnt, bone2Name=Comb_bk_02_jnt, swing1LimitDegrees=15, swing2LimitDegrees=15, twistLimitDegrees=2}
 {
  bone1Name=Comb_bk_tp_MSTR
  bone2Name=head
  swing1LimitDegrees=45
  swing1Motion=Free
  swing2LimitDegrees=45
  swing2Motion=Free
  twistLimitDegrees=45
  twistMotion=Free
 }
 {bone1Name=Comb_bk_tp_01_jnt, bone2Name=Comb_bk_tp_MSTR, swing1LimitDegrees=15, swing2LimitDegrees=15, twistLimitDegrees=2}
 {bone1Name=Comb_bk_tp_02_jnt, bone2Name=Comb_bk_tp_01_jnt, swing1LimitDegrees=15, swing2LimitDegrees=15, twistLimitDegrees=2}
 {bone1Name=Comb_bk_tp_03_jnt, bone2Name=Comb_bk_tp_02_jnt, swing1LimitDegrees=15, swing2LimitDegrees=15, twistLimitDegrees=2}
 {bone1Name=Comb_bk_tp_04_jnt, bone2Name=Comb_bk_tp_03_jnt, swing1LimitDegrees=15, swing2LimitDegrees=15, twistLimitDegrees=2}
 {
  bone1Name=Comb_ft_MSTR
  bone2Name=head
  swing1LimitDegrees=45
  swing1Motion=Free
  swing2LimitDegrees=45
  swing2Motion=Free
  twistLimitDegrees=45
  twistMotion=Free
 }
 {bone1Name=Comb_ft_01_jnt, bone2Name=Comb_ft_MSTR, swing1LimitDegrees=15, swing2LimitDegrees=15, twistLimitDegrees=2}
 {bone1Name=Comb_ft_02_jnt, bone2Name=Comb_ft_01_jnt, swing1LimitDegrees=15, swing2LimitDegrees=15, twistLimitDegrees=2}
 {
  bone1Name=Comb_ft_tp_MSTR
  bone2Name=head
  swing1LimitDegrees=45
  swing1Motion=Free
  swing2LimitDegrees=45
  swing2Motion=Free
  twistLimitDegrees=45
  twistMotion=Free
 }
 {bone1Name=Comb_ft_tp_01_jnt, bone2Name=Comb_ft_tp_MSTR, swing1LimitDegrees=15, swing2LimitDegrees=15, twistLimitDegrees=2}
 {bone1Name=Comb_ft_tp_02_jnt, bone2Name=Comb_ft_tp_01_jnt, swing1LimitDegrees=15, swing2LimitDegrees=15, twistLimitDegrees=2}
 {bone1Name=Comb_ft_tp_03_jnt, bone2Name=Comb_ft_tp_02_jnt, swing1LimitDegrees=15, swing2LimitDegrees=15, twistLimitDegrees=2}
 {bone1Name=Comb_ft_tp_04_jnt, bone2Name=Comb_ft_tp_03_jnt, swing1LimitDegrees=15, swing2LimitDegrees=15, twistLimitDegrees=2}
 {bone1Name=Comb_ft_tp_05_jnt, bone2Name=Comb_ft_tp_04_jnt, swing1LimitDegrees=15, swing2LimitDegrees=15, twistLimitDegrees=2}
 {bone1Name=Comb_ft_tp_06_jnt, bone2Name=Comb_ft_tp_05_jnt, swing1LimitDegrees=15, swing2LimitDegrees=15, twistLimitDegrees=2}
 {
  bone1Name=C_jaw
  bone2Name=head
  swing1LimitDegrees=45
  swing1Motion=Free
  swing2LimitDegrees=45
  swing2Motion=Free
  twistLimitDegrees=45
  twistMotion=Free
 }
 {
  bone1Name=Wattle_l_MSTR
  bone2Name=C_jaw
  swing1LimitDegrees=45
  swing1Motion=Free
  swing2LimitDegrees=45
  swing2Motion=Free
  twistLimitDegrees=45
  twistMotion=Free
 }
 {bone1Name=Wattle_l_01_jnt, bone2Name=Wattle_l_MSTR, swing1LimitDegrees=25, swing2LimitDegrees=25, twistLimitDegrees=2}
 {bone1Name=Wattle_l_02_jnt, bone2Name=Wattle_l_01_jnt, swing1LimitDegrees=25, swing2LimitDegrees=25, twistLimitDegrees=2}
 {bone1Name=Wattle_l_03_jnt, bone2Name=Wattle_l_02_jnt, swing1LimitDegrees=25, swing2LimitDegrees=25, twistLimitDegrees=2}
 {
  bone1Name=Wattle_r_MSTR
  bone2Name=C_jaw
  swing1LimitDegrees=45
  swing1Motion=Free
  swing2LimitDegrees=45
  swing2Motion=Free
  twistLimitDegrees=45
  twistMotion=Free
 }
 {bone1Name=Wattle_r_01_jnt, bone2Name=Wattle_r_MSTR, swing1LimitDegrees=25, swing2LimitDegrees=25, twistLimitDegrees=2}
 {bone1Name=Wattle_r_02_jnt, bone2Name=Wattle_r_01_jnt, swing1LimitDegrees=25, swing2LimitDegrees=25, twistLimitDegrees=2}
 {bone1Name=Wattle_r_03_jnt, bone2Name=Wattle_r_02_jnt, swing1LimitDegrees=25, swing2LimitDegrees=25, twistLimitDegrees=2}
 {
  bone1Name=head
  bone2Name=spine_05
  swing1LimitDegrees=45
  swing1Motion=Free
  swing2LimitDegrees=45
  swing2Motion=Free
  twistLimitDegrees=45
  twistMotion=Free
 }
]
